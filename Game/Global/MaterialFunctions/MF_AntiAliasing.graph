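# MF_AntiAliasing is a TechArt library material function that turns a scalar gradient into an anti-aliased mask around a cutoff.
# It divides CutOff minus Gradient by the length of that difference's screen-space DDX/DDY derivatives plus Width, keeping the edge about one pixel wide at zero Width.
# The result is subtracted from 0.5 and saturated, so values above the cutoff go toward white and larger Width softens the transition.
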MaterialFunction /Game/Global/MaterialFunctions/MF_AntiAliasing
graph {
 properties {bExposeToLibrary=true, libraryCategoriesText [TechArt]}
 nodes [
  n1 = FunctionInput() {
   bUsePreviewValueAsDefault=true
   desc="Default = 0.5"
   inputName=CutOff
   inputType=FunctionInput_Scalar
   sortPriority=1
   previewValue=<0.5 0 0 1>
  }
  n2 = Subtract(
   A=@n1
   B=FunctionInput(Preview=MaterialFunctionCall() {materialFunction=&LinearGradient}.UGradient) {bUsePreviewValueAsDefault=true, inputName=Gradient, inputType=FunctionInput_Scalar}
  )
  n3 = AppendVector(A=DDX(Value=@n2), B=DDY(Value=@n2))
  n4 = FunctionInput() {
   bUsePreviewValueAsDefault=true
   desc="0.0 = sharp\r\n0.001 = small transition\r\n0.01 = big transition"
   inputName=Width
   inputType=FunctionInput_Scalar
   sortPriority=2
  }
  n5 = Divide(A=@n2, B=Add(A=Length(None=@n3) {is=class MaterialExpressionLength}, B=@n4) {constB=0})
  FunctionOutput(None=Saturate(None=Subtract(B=@n5) {constA=0.5}))
 ]
}
