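# MF_FloatEqual is a material function in the Math library category that returns 1.0 when scalar inputs A and B are equal, else 0.0.
# An If node compares A and B, outputting a constant 1 on equality and a shared constant 0 for both the greater and less branches.
# Input B falls back to its preview value when left unconnected.
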MaterialFunction /Game/Global/MaterialFunctions/MF_FloatEqual
graph {
 properties {
  bExposeToLibrary=true
  description="Returns 1.0 if A and B are Equal, else returns 0.0"
  userExposedCaption="f1Equal =="
  libraryCategoriesText [Math]
 }
 nodes [
  n1 = Constant()
  FunctionOutput(None=If(
    A=FunctionInput() {inputName=A, inputType=FunctionInput_Scalar, previewValue=<1 0 0 1>}
    B=FunctionInput() {bUsePreviewValueAsDefault=true, inputName=B, inputType=FunctionInput_Scalar, sortPriority=1}
    "A > B"=@n1
    "A == B"=Constant() {r=1}
    "A < B"=@n1
   )) {outputName=Return}
 ]
}
