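# Material function "f2Equal ==", exposed in the Math library, that compares two Vector2 inputs A and B for component-wise equality, with B optional.
# It breaks both inputs into scalars, feeds them to two MF_FloatEqual calls and packs the results into a Vector2 Return.
# Two extra outputs reduce those results: All takes their Min, while Any saturates their sum, avoiding branches.
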MaterialFunction /Game/Global/MaterialFunctions/MF_Float2Equal
graph {
 expressions {
  MaterialExpressionAdd_0 {
   class=MaterialExpressionAdd
   inputs [
    A {from=MaterialExpressionMaterialFunctionCall_0, output=Return}
    B {from=MaterialExpressionMaterialFunctionCall_1, output=Return}
   ]
  }
  MaterialExpressionConvert_0 {
   class=MaterialExpressionConvert
   inputs [None {from=MaterialExpressionFunctionInput_0}]
   properties {
    nodeName=Break Float2
    convertInputs [{
      type=Vector2
      expressionInput {
       expression=&self:MaterialExpressionFunctionInput_0
       inputName=None
       mask=0
       maskA=0
       maskB=0
       maskG=0
       maskR=0
       outputIndex=0
      }
     }]
    convertOutputs [{type=Scalar}, {type=Scalar}]
   }
  }
  MaterialExpressionConvert_1 {
   class=MaterialExpressionConvert
   inputs [
    None {from=MaterialExpressionMaterialFunctionCall_0, output=Return}
    None {from=MaterialExpressionMaterialFunctionCall_1, output=Return}
   ]
   properties {
    nodeName=Make Vector2
    convertInputs [
     {
      type=Scalar
      expressionInput {
       expression=&self:MaterialExpressionMaterialFunctionCall_0
       inputName=None
       mask=0
       maskA=0
       maskB=0
       maskG=0
       maskR=0
       outputIndex=0
      }
     }
     {
      type=Scalar
      expressionInput {
       expression=&self:MaterialExpressionMaterialFunctionCall_1
       inputName=None
       mask=0
       maskA=0
       maskB=0
       maskG=0
       maskR=0
       outputIndex=0
      }
     }
    ]
    convertOutputs [{type=Vector2}]
   }
  }
  MaterialExpressionConvert_2 {
   class=MaterialExpressionConvert
   inputs [None {from=MaterialExpressionFunctionInput_1}]
   properties {
    nodeName=Break Float2
    convertInputs [{
      type=Vector2
      expressionInput {
       expression=&self:MaterialExpressionFunctionInput_1
       inputName=None
       mask=0
       maskA=0
       maskB=0
       maskG=0
       maskR=0
       outputIndex=0
      }
     }]
    convertOutputs [{type=Scalar}, {type=Scalar}]
   }
  }
  MaterialExpressionFunctionInput_0 {
   class=MaterialExpressionFunctionInput
   inputs [Preview {from="", output=""}]
   properties {inputName=A, inputType=FunctionInput_Vector2, previewValue=<1 0 0 1>}
  }
  MaterialExpressionFunctionInput_1 {
   class=MaterialExpressionFunctionInput
   inputs [Preview {from="", output=""}]
   properties {bUsePreviewValueAsDefault=true, inputName=B, inputType=FunctionInput_Vector2, sortPriority=1}
  }
  MaterialExpressionFunctionOutput_0 {
   class=MaterialExpressionFunctionOutput
   inputs [None {from=MaterialExpressionConvert_1, output=""}]
   properties {description=Returns component-wise comparison, outputName=Return}
  }
  MaterialExpressionFunctionOutput_1 {
   class=MaterialExpressionFunctionOutput
   inputs [None {from=MaterialExpressionMin_0, output=""}]
   properties {description=Returns 1.0 if all components are equal, outputName=All, sortPriority=1}
  }
  MaterialExpressionFunctionOutput_2 {
   class=MaterialExpressionFunctionOutput
   inputs [None {from=MaterialExpressionSaturate_1, output=""}]
   properties {description=Returns 1.0 if any two components are equal, outputName=Any, sortPriority=2}
  }
  MaterialExpressionMaterialFunctionCall_0 {
   class=MaterialExpressionMaterialFunctionCall
   inputs [
    A {from=MaterialExpressionConvert_0}
    B {from=MaterialExpressionConvert_2}
   ]
   properties {materialFunction=&/MF_FloatEqual}
  }
  MaterialExpressionMaterialFunctionCall_1 {
   class=MaterialExpressionMaterialFunctionCall
   inputs [
    A {from=MaterialExpressionConvert_0}
    B {from=MaterialExpressionConvert_2}
   ]
   properties {materialFunction=&MF_FloatEqual}
  }
  MaterialExpressionMin_0 {
   class=MaterialExpressionMin
   inputs [
    A {from=MaterialExpressionMaterialFunctionCall_0, output=Return}
    B {from=MaterialExpressionMaterialFunctionCall_1, output=Return}
   ]
   properties {}
  }
  MaterialExpressionSaturate_1 {class=MaterialExpressionSaturate, inputs [None {from=MaterialExpressionAdd_0}]}
 }
 properties {
  bExposeToLibrary=true
  description="Returns component-wise 1.0 if A and B are Equal, else returns 0.0. Optionally returns 1.0 if all or any components are equal"
  userExposedCaption="f2Equal =="
  libraryCategoriesText [Math]
 }
}
