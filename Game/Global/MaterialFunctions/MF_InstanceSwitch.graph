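# Material function exposed as InstanceSwitchDynamic that chooses between Instanced and Default inputs, intended to vary materials on instanced primitives.
# A Custom HLSL node tests primitive flags against the instance-local-bounds flag, using logical && rather than bitwise &, so any nonzero flags pass.
# An If compares that result to 0.5, feeding f3 directly and f1 through a red-channel mask.
# The switch is dynamic, adding no permutations but evaluating both paths.
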MaterialFunction /Game/Global/MaterialFunctions/MF_InstanceSwitch
graph {
 properties {
  bExposeToLibrary=true
  description="Allows material to define behaviour if primitive is instanced. This switch is NOT STATIC and does not add permutations; however both paths are fully evaluated. "
  userExposedCaption=InstanceSwitchDynamic
  libraryCategoriesText [Global Function Library]
 }
 nodes [
  n1 = If(
   A=Custom!(None) {code="return (GetPrimitiveData(Parameters).Flags && PRIMITIVE_SCENE_DATA_FLAG_HAS_INSTANCE_LOCAL_BOUNDS) != 0 ? 1 : 0;", showCode=true}
   "A > B"=FunctionInput() {description=Input to be used with instanced mesh, inputName=Instanced, sortPriority=1}
   "A < B"=FunctionInput() {description=Input to be used with non instanced mesh, inputName=Default}
  ) {constB=0.5}
  FunctionOutput(None=ComponentMask(None=@n1) {r=true}) {description=Scalar result, outputName=f1}
  FunctionOutput(None=@n1) {description=Vec3 result, outputName=f3, sortPriority=1}
 ]
}
