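# MaxComponents is a material function that returns the largest component of a vector, exposed to the material library under Math.
# It has three independent input and output pairs, f2, f3 and f4, for two, three and four component vectors.
# Each input is split into scalars by a Convert node, then reduced to one value through chained Max nodes.
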MaterialFunction /Game/Global/MaterialFunctions/MF_MaxComponents
graph {
 properties {
  bExposeToLibrary=true
  description=Returns the maximum value from the given components
  userExposedCaption=MaxComponents
  libraryCategoriesText [Math]
 }
 nodes [
  MaterialExpressionFunctionInput_1 = FunctionInput() {bUsePreviewValueAsDefault=true, inputName=f2, inputType=FunctionInput_Vector2}
  n1 = Convert(None=@MaterialExpressionFunctionInput_1) {
   nodeName=Break Float2
   convertInputs [{
     type=Vector2
     expressionInput {
      expression=&self:MaterialExpressionFunctionInput_1
      inputName=None
      mask=0
      maskA=0
      maskB=0
      maskG=0
      maskR=0
      outputIndex=0
     }
    }]
   convertOutputs [{type=Scalar}, {type=Scalar}]
  }
  FunctionOutput(None=Max(A=@n1, B=@n1)) {outputName=f2}
  MaterialExpressionFunctionInput_2 = FunctionInput() {bUsePreviewValueAsDefault=true, inputName=f3, sortPriority=1}
  n2 = Convert(None=@MaterialExpressionFunctionInput_2) {
   nodeName=Break Float3
   convertInputs [{
     type=Vector3
     expressionInput {
      expression=&self:MaterialExpressionFunctionInput_2
      inputName=None
      mask=0
      maskA=0
      maskB=0
      maskG=0
      maskR=0
      outputIndex=0
     }
    }]
   convertOutputs [{type=Scalar}, {type=Scalar}, {type=Scalar}]
  }
  FunctionOutput(None=Max(A=Max(A=@n2, B=@n2), B=@n2)) {outputName=f3, sortPriority=1}
  MaterialExpressionFunctionInput_3 = FunctionInput() {bUsePreviewValueAsDefault=true, inputName=f4, inputType=FunctionInput_Vector4, sortPriority=2}
  n3 = Convert(None=@MaterialExpressionFunctionInput_3) {
   nodeName=Break Float4
   convertInputs [{
     type=Vector4
     defaultValue=<rgba 0 0 0 1>
     expressionInput {
      expression=&self:MaterialExpressionFunctionInput_3
      inputName=None
      mask=0
      maskA=0
      maskB=0
      maskG=0
      maskR=0
      outputIndex=0
     }
    }]
   convertOutputs [{type=Scalar}, {type=Scalar}, {type=Scalar}, {type=Scalar}]
  }
  n4 = Max(A=@n3, B=@n3)
  FunctionOutput(None=Max(A=Max(A=@n4, B=@n3), B=@n3)) {outputName=f4, sortPriority=2}
 ]
}
