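# Material function that computes horizon occlusion, a scalar that fades out reflections pointing below the geometric surface, as normal mapping can cause.
# It dots a direction input, defaulting to the world reflection vector, with the vertex normal and scales the result by Horizon Fade.
# The scaled value is offset by the Add node's default constant, saturated and squared for a soft falloff.
# The direction input also accepts an inverted refraction vector.
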MaterialFunction /Game/Global/MaterialFunctions/MF_NormalHorizonOcclusion
graph {nodes [
  n1 = FunctionInput() {bUsePreviewValueAsDefault=true, inputName=Horizon Fade, inputType=FunctionInput_Scalar, sortPriority=10, previewValue=<1 0 0 1>}
  n2 = FunctionInput(Preview=ReflectionVectorWS()) {
   bUsePreviewValueAsDefault=true
   description="Reflection vector or inverted refraction vector (default : ReflectionVectorWS)"
   inputName=Direction WS
  }
  n3 = DotProduct(A=@n2, B=VertexNormalWS())
  n4 = Saturate(None=Add(A=Multiply(A=@n1, B=@n3)))
  FunctionOutput(None=Multiply(A=@n4, B=@n4)) {outputName="Occlusion (S)"}
 ]}
