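# Material function for biplanar texture mapping, outputting Color and Normals, based on Inigo Quilez's technique.
# Two Custom HLSL nodes project along the normal's two dominant axes with explicit gradients and blend them, using pre-skinned position and normal by default.
# The normal path adds strength control and converts to tangent space from local or world via a static switch, defaulting to world.
# MaterialProxyReplace substitutes WorldAlignedTexture for proxy baking.
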MaterialFunction /Game/Global/MaterialFunctions/MF_BiPlanar
graph {
 properties {bExposeToLibrary=true}
 nodes [
  MaterialExpressionNamedRerouteUsage_101 = NamedRerouteUsage()
  n1 = FunctionInput(Preview=VertexInterpolator(VS=PreSkinnedPosition()).PS) {bUsePreviewValueAsDefault=true, inputName=Pos}
  n2 = FunctionInput(Preview=Constant() {r=100}) {bUsePreviewValueAsDefault=true, inputName=ColorUVScale, inputType=FunctionInput_Scalar, sortPriority=2}
  MaterialExpressionDivide_2 = Divide(A=@n1, B=@n2)
  n3 = PreSkinnedNormal()
  MaterialExpressionFunctionInput_3 = FunctionInput(Preview=ComponentMask(None=VertexInterpolator(VS=@n3).PS) {b=true, g=true, r=true}) {bUsePreviewValueAsDefault=true, inputName=Nrm, sortPriority=1}
  MaterialExpressionFunctionInput_4 = FunctionInput(Preview=Constant() {r=8}) {bUsePreviewValueAsDefault=true, inputName=ColorBlendFactor, inputType=FunctionInput_Scalar, sortPriority=3}
  n4 = Custom!(
   colorTex=@MaterialExpressionNamedRerouteUsage_101
   wpos=@MaterialExpressionDivide_2
   wnrm=@MaterialExpressionFunctionInput_3
   k=@MaterialExpressionFunctionInput_4
  ) {
   code="// Copyright © 2020 Inigo Quilez\r\n// https://iquilezles.org/www/articles/biplanar/biplanar.htm\r\n\r\nfloat3 p = wpos;\r\nfloat3 n = abs((wnrm));\r\n\r\n// grab coord derivatives for texturing\r\nfloat3 dpdx = ddx(p);\r\nfloat3 dpdy = ddy(p);\r\n\r\n// major axis (in x; yz are following axis)\r\nuint3 ma = (n.x > n.y &&&& n.x > n.z) ? uint3(0, 1, 2) :\r\n\t\t   (n.y > n.z)\t\t\t\t? uint3(1, 2, 0) :\r\n\t\t\t\t\t\t\t\t\t  uint3(2, 0, 1) ;\r\n\t\t\t\t\t\t\t\t\t  \r\n// minor axis (in x; yz are following axis)\r\nuint3 mi = (n.x < n.y &&&& n.x < n.z) ? uint3(0, 1, 2) :\r\n\t\t   (n.y < n.z)\t\t\t\t? uint3(1, 2, 0) :\r\n\t\t\t\t\t\t\t\t\t  uint3(2, 0, 1) ;\r\n\r\n// median axis (in x; yz are following axis)\r\nuint3 me = 3 - mi - ma;\r\nme = max(me, uint3(0,0,0));\r\nme = min(me, uint3(2,2,2));\r\n// project + fetch\r\nfloat3 x = (Texture2DSampleGrad(colorTex, colorTexSampler,\r\n\t\t\t\t\t\t\t\t float2(   p[ma.y],    p[ma.z]), \r\n\t\t\t\t\t\t\t\t float2(dpdx[ma.y], dpdx[ma.z]), \r\n\t\t\t\t\t\t\t\t float2(dpdy[ma.y], dpdy[ma.z]))).xyz;\r\n\r\nfloat3 y = (Texture2DSampleGrad(colorTex, colorTexSampler,\r\n\t\t\t\t\t\t\t\t float2(   p[me.y],    p[me.z]), \r\n\t\t\t\t\t\t\t\t float2(dpdx[me.y], dpdx[me.z]),\r\n\t\t\t\t\t\t\t\t float2(dpdy[me.y], dpdy[me.z]))).xyz;\r\n\r\n// blend and return\r\nfloat2 m = float2(n[ma.x], n[me.x]);\r\n\r\n// optional - add local support (prevents discontinuty)\r\nm = saturate((m - 0.5773f) / (1.0f - 0.5773f));\r\n\r\n// transition control\r\nm = pow( m, k/8.0);\r\nreturn ((x * m.x + y * m.y) / (m.x + m.y));\r\n\r\n"
   description=BiPlanar Mapping
   inputs [
    {inputName=colorTex, input {expression=&self:MaterialExpressionNamedRerouteUsage_101}}
    {inputName=wpos, input {expression=&self:MaterialExpressionDivide_2}}
    {inputName=wnrm, input {expression=&self:MaterialExpressionFunctionInput_3}}
    {inputName=k, input {expression=&self:MaterialExpressionFunctionInput_4}}
   ]
  }
  n5 = Multiply(A=NamedRerouteUsage()) {constB=1.5}
  n6 = Constant() {r=0.25}
  FunctionOutput(None=MaterialProxyReplace(
    Realtime=@n4
    MaterialProxy=MaterialFunctionCall(TextureObject=@MaterialExpressionNamedRerouteUsage_101, TextureSize=@n5, RotZAxis=@n6, RotXAxis=@n6, RotYAxis=@n6) {materialFunction=/Game/Weapons/WeaponSkins/Wraps/MaterialFunctions/WorldAlignedTexture}.XYZ Texture
   )) {outputName=Color, sortPriority=-1}
  MaterialExpressionNamedRerouteUsage_102 = NamedRerouteUsage()
  n7 = FunctionInput(Preview=Constant() {r=100}) {bUsePreviewValueAsDefault=true, inputName=NormalUVScale, inputType=FunctionInput_Scalar, sortPriority=11}
  MaterialExpressionDivide_0 = Divide(A=@n1, B=@n7)
  MaterialExpressionFunctionInput_7 = FunctionInput(Preview=Constant() {r=8}) {bUsePreviewValueAsDefault=true, inputName=NormalBlendFactor, inputType=FunctionInput_Scalar, sortPriority=12}
  MaterialExpressionFunctionInput_8 = FunctionInput(Preview=Constant() {r=1}) {bUsePreviewValueAsDefault=true, inputName=NormalStrength, inputType=FunctionInput_Scalar, sortPriority=10}
  n8 = Custom!(
   normTex=@MaterialExpressionNamedRerouteUsage_102
   wpos=@MaterialExpressionDivide_0
   wnrm=@MaterialExpressionFunctionInput_3
   k=@MaterialExpressionFunctionInput_7
   strength=@MaterialExpressionFunctionInput_8
  ) {
   code="// Copyright © 2020 Inigo Quilez\r\n// https://iquilezles.org/www/articles/biplanar/biplanar.htm\r\n\r\nfloat3 p = wpos;\r\nfloat3 n = abs((wnrm));\r\n\r\n// grab coord derivatives for texturing\r\nfloat3 dpdx = ddx(p);\r\nfloat3 dpdy = ddy(p);\r\n\r\n// major axis (in x; yz are following axis)\r\nuint3 ma = (n.x > n.y &&&& n.x > n.z) ? uint3(0, 1, 2) :\r\n\t\t   (n.y > n.z)\t\t\t\t? uint3(1, 2, 0) :\r\n\t\t\t\t\t\t\t\t\t  uint3(2, 0, 1) ;\r\n\t\t\t\t\t\t\t\t\t  \r\n// minor axis (in x; yz are following axis)\r\nuint3 mi = (n.x < n.y &&&& n.x < n.z) ? uint3(0, 1, 2) :\r\n\t\t   (n.y < n.z)\t\t\t\t? uint3(1, 2, 0) :\r\n\t\t\t\t\t\t\t\t\t  uint3(2, 0, 1) ;\r\n\r\n// median axis (in x; yz are following axis)\r\nuint3 me = 3 - mi - ma;\r\n\r\nme = max(me, uint3(0,0,0));\r\nme = min(me, uint3(2,2,2));\r\n\r\n// project + fetch\r\nfloat3 x = (Texture2DSampleGrad(normTex, normTexSampler,\r\n\t\t\t\t\t\t\t\t float2(   p[ma.y],    p[ma.z]), \r\n\t\t\t\t\t\t\t\t float2(dpdx[ma.y], dpdx[ma.z]), \r\n\t\t\t\t\t\t\t\t float2(dpdy[ma.y], dpdy[ma.z]))).xyz;\r\n\t\t\t\t\t\t\t\t \r\nfloat3 y = (Texture2DSampleGrad(normTex, normTexSampler,\r\n\t\t\t\t\t\t\t\t float2(   p[me.y],    p[me.z]), \r\n\t\t\t\t\t\t\t\t float2(dpdx[me.y], dpdx[me.z]),\r\n\t\t\t\t\t\t\t\t float2(dpdy[me.y], dpdy[me.z]))).xyz;\r\n\t\t\t\t\t\t\t\t \r\n// Normal vector extraction\r\nfloat3 n1; \r\nn1.xy = x.rg * 2 - 1;\r\nn1.z = max(sqrt(1.0f - saturate(dot(n1.xy, n1.xy))), .001);\r\n\r\nfloat3 n2;\r\nn2.xy = y.rg * 2 - 1;\r\nn2.z = max(sqrt(1.0f - saturate(dot(n2.xy, n2.xy))), .001);\r\n\r\nn1 = normalize(n1 * float3(strength,strength,1.0));\r\nn2 = normalize(n2 * float3(strength,strength,1.0));\r\n\r\nn1 = normalize(float3(n1.y + wnrm[ma.z], n1.x + wnrm[ma.y], wnrm[ma.x]));\r\nn2 = normalize(float3(n2.y + wnrm[me.z], n2.x + wnrm[me.y], wnrm[me.x]));\r\nn1 = float3(n1[ma.z], n1[ma.y], n1[ma.x]);\r\nn2 = float3(n2[me.z], n2[me.y], n2[me.x]);\r\n\r\n\r\n// blend and return\r\nfloat2 m = float2(n[ma.x], n[me.x]);\r\n\r\n// optional - add local support (prevents discontinuty)\r\nm = saturate((m - 0.5773f) / (1.0f - 0.5773f));\r\n\r\n// transition control\r\nm = pow( m, k/8.0);\r\nm = max(m, .001);\r\nfloat3 wn = normalize((n1 * m.x + n2 * m.y) / (m.x + m.y));\r\n//return mul((MaterialFloat3x3)(Parameters.TangentToWorld), wn);\r\nreturn wn;"
   description=BiPlanar Normals
   inputs [
    {inputName=normTex, input {expression=&self:MaterialExpressionNamedRerouteUsage_102}}
    {inputName=wpos, input {expression=&self:MaterialExpressionDivide_0}}
    {inputName=wnrm, input {expression=&self:MaterialExpressionFunctionInput_3}}
    {inputName=k, input {expression=&self:MaterialExpressionFunctionInput_7}}
    {inputName=strength, input {expression=&self:MaterialExpressionFunctionInput_8}}
   ]
  }
  n9 = Transform(None=@n8) {transformSourceType=TRANSFORMSOURCE_Local, transformType=TRANSFORM_Tangent}
  n10 = Transform(None=@n8) {transformSourceType=TRANSFORMSOURCE_World, transformType=TRANSFORM_Tangent}
  n11 = FunctionInput(Preview=StaticBool()) {bUsePreviewValueAsDefault=true, inputName=Local Space Inputs, inputType=FunctionInput_StaticBool, sortPriority=50}
  n12 = NamedRerouteUsage()
  n13 = Multiply(A=NamedRerouteUsage()) {constB=1.5}
  n14 = Constant() {r=0.25}
  FunctionOutput(None=MaterialProxyReplace(
    Realtime=StaticSwitch(True=@n9, False=@n10, Value=@n11)
    MaterialProxy=MaterialFunctionCall(TextureObject=@n12, TextureSize=@n13, RotZAxis=@n14, RotXAxis=@n14, RotYAxis=@n14) {materialFunction=&WorldAlignedTexture}.XYZ Texture
   )) {outputName=Normals}
  NamedRerouteDeclaration(None=@n2) {name=scale_Color, nodeColor=<rgba 0.2188 0 1 1>}
  NamedRerouteDeclaration(None=@n7) {name=scale_Normal, nodeColor=<rgba 0.1719 0 1 1>}
  n15 = FunctionInput(Preview=TextureObject() {samplerType=SAMPLERTYPE_Normal, texture=/Game/Athena/Apollo/Water/Textures/02_Rounded/03_WaterFoam_AccentSwirls_N}) {bUsePreviewValueAsDefault=true, inputName=NormalTex, inputType=FunctionInput_Texture2D, sortPriority=-1, previewValue=<0 0 1 1>}
  NamedRerouteDeclaration(None=@n15) {name=texNorm, nodeColor=<rgba 0 0.0859 1 1>}
  n16 = FunctionInput(Preview=TextureObject() {samplerType=SAMPLERTYPE_LinearColor, texture=/Game/Athena/Apollo/Water/Textures/02_Rounded/03_WaterFoam_Veroni_base_H}) {bUsePreviewValueAsDefault=true, inputName=ColorTex, inputType=FunctionInput_Texture2D, sortPriority=-2, previewValue=<0 0 1 1>}
  NamedRerouteDeclaration(None=@n16) {name=tex, nodeColor=<rgba 0 0.0859 1 1>}
 ]
}
