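# MF_Bias is a library-exposed material function that remaps a scalar value along a bias curve.
# It saturates the value input and clamps bias just inside 0 to 1 to keep the division safe, passing both through named reroutes.
# The math matches Schlick's bias approximation, value / ((1/bias - 2)(1 - value) + 1), avoiding a power function.
# Bias previews at a neutral 0.5, and value previews as horizontal UV.
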MaterialFunction /Game/Global/MaterialFunctions/MF_Bias
graph {
 properties {bExposeToLibrary=true}
 nodes [
  n1 = NamedRerouteUsage()
  n2 = Multiply(A=Subtract(A=Divide(B=@n1)) {constB=2}, B=Subtract(B=NamedRerouteUsage()))
  FunctionOutput(None=Divide(A=NamedRerouteUsage(), B=Add(A=@n2)))
  n3 = ComponentMask(None=TextureCoordinate()) {r=true}
  NamedRerouteDeclaration(None=Saturate(None=FunctionInput(Preview=@n3) {bUsePreviewValueAsDefault=true, inputName=value, inputType=FunctionInput_Scalar})) {name=value, nodeColor=<rgba 1 0 0.8437 1>}
  n4 = FunctionInput() {bUsePreviewValueAsDefault=true, inputName=bias, inputType=FunctionInput_Scalar, sortPriority=1, previewValue=<0.5 0 0 1>}
  NamedRerouteDeclaration(None=Clamp(None=@n4) {maxDefault=0.9999, minDefault=0.0001}) {name=bias, nodeColor=<rgba 1 0 0.7734 1>}
 ]
}
