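# MF_NormalBlend_Whiteout is a material function that combines two tangent-space normals using whiteout blending.
# It adds the red and green channels of Normal_1 and Normal_2, multiplies their blue channels, and appends the results into one vector.
# Both inputs default to a flat normal, and the output is not normalized inside the function.
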MaterialFunction /Game/Global/MaterialFunctions/MF_NormalBlend_Whiteout
graph {nodes [
  n1 = FunctionInput() {bUsePreviewValueAsDefault=true, inputName=Normal_1, sortPriority=-10, previewValue=<0 0 1 1>}
  n2 = ComponentMask(None=@n1) {g=true, r=true}
  n3 = FunctionInput() {bUsePreviewValueAsDefault=true, inputName=Normal_2, previewValue=<0 0 1 1>}
  n4 = ComponentMask(None=@n3) {g=true, r=true}
  n5 = ComponentMask(None=@n1) {b=true}
  n6 = ComponentMask(None=@n3) {b=true}
  FunctionOutput(None=AppendVector(A=Add(A=@n2, B=@n4), B=Multiply(A=@n5, B=@n6)))
 ]}
