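# Material function that blends a normal texture with world normals using whiteout blending, with adjustable flattening.
# It passes both normal inputs to MF_NormalBlend_Whiteout, then lerps the result back toward the world normals by the NormalFlatness input.
# A manual XY-add-and-normalize blend and a NormalFlatness scalar parameter sit in the graph unconnected to the output.
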MaterialFunction /Game/Global/MaterialFunctions/MF_BlendAxisNormals_Whiteout
graph {nodes [
  n1 = FunctionInput() {inputName=WorldNormals}
  n2 = FunctionInput() {inputName=NormalTexture, sortPriority=-1}
  FunctionOutput(None=LinearInterpolate(
    A=MaterialFunctionCall!(Normal_1=@n1, Normal_2=@n2) {materialFunction=&/MF_NormalBlend_Whiteout}.Result
    B=@n1
    Alpha=FunctionInput() {bUsePreviewValueAsDefault=true, inputName=NormalFlatness, inputType=FunctionInput_Scalar}
   ))
  n3 = Constant3Vector() {constant=<rgba 1 1 0 0>}
  Normalize(VectorInput=Add(A=Multiply(A=@n2, B=@n3), B=@n1))
  ScalarParameter() {parameterName=NormalFlatness}
 ]}
