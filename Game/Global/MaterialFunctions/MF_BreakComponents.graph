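# MF_BreakComponents is a TechArt library material function that splits a Float4 input into its individual channels and common combinations.
# The single Vector4 input, which uses its preview value as default, feeds a separate component mask for each output.
# Outputs are sorted as RGB, RG, R, G, B, A and RGBA, exposing both combined swizzles and single channels from one node.
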MaterialFunction /Game/Global/MaterialFunctions/MF_BreakComponents
graph {
 properties {bExposeToLibrary=true, libraryCategoriesText [TechArt]}
 nodes [
  n1 = FunctionInput() {bUsePreviewValueAsDefault=true, inputName=Float4, inputType=FunctionInput_Vector4}
  FunctionOutput(None=ComponentMask(None=@n1) {g=true, r=true}) {outputName=RG, sortPriority=1}
  FunctionOutput(None=ComponentMask(None=@n1) {r=true}) {outputName=R, sortPriority=2}
  FunctionOutput(None=ComponentMask(None=@n1) {g=true}) {outputName=G, sortPriority=3}
  FunctionOutput(None=ComponentMask(None=@n1) {b=true, g=true, r=true}) {outputName=RGB}
  FunctionOutput(None=ComponentMask(None=@n1) {b=true}) {outputName=B, sortPriority=4}
  FunctionOutput(None=ComponentMask(None=@n1) {a=true, b=true, g=true, r=true}) {outputName=RGBA, sortPriority=6}
  FunctionOutput(None=ComponentMask(None=@n1) {a=true}) {outputName=A, sortPriority=5}
 ]
}
