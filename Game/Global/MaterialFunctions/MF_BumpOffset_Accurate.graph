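# Material function that offsets UVs for a bump offset parallax effect, exposed to the library under Tech Art and Misc.
# It transforms the camera vector to tangent space and reflects it about the Normal input via CustomReflectionVector.
# Height is divided by the absolute Z of that vector before ReferencePlane is subtracted.
# The result scales the vector's XY and HeightRatio, then is added to the input UVs.
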MaterialFunction /Game/Global/MaterialFunctions/MF_BumpOffset_Accurate
graph {
 properties {bExposeToLibrary=true, libraryCategoriesText [Tech Art, Misc]}
 nodes [
  n1 = MaterialFunctionCall!(
   Normal=FunctionInput() {bUsePreviewValueAsDefault=true, inputName=Normal, sortPriority=4, previewValue=<0 0 1 1>}
   CameraVector=Transform(None=CameraVectorWS()) {transformSourceType=TRANSFORMSOURCE_World, transformType=TRANSFORM_Tangent}
  ) {materialFunction=/Engine/Functions/Engine_MaterialFunctions02/Utility/CustomReflectionVector}
  n2 = FunctionInput() {bUsePreviewValueAsDefault=true, inputName=Height, inputType=FunctionInput_Scalar, sortPriority=1, previewValue=<100 0 0 1>}
  n3 = Abs(None=ComponentMask(None=@n1.Reflection Vector) {b=true})
  n4 = FunctionInput() {bUsePreviewValueAsDefault=true, inputName=ReferencePlane, inputType=FunctionInput_Scalar, sortPriority=2, previewValue=<0.5 0 0 1>}
  n5 = Multiply(A=ComponentMask(None=@n1.Reflection Vector) {g=true, r=true}, B=Subtract(A=Divide(A=@n2, B=@n3), B=@n4))
  n6 = FunctionInput() {bUsePreviewValueAsDefault=true, inputName=HeightRatio, inputType=FunctionInput_Scalar, sortPriority=3, previewValue=<0.05 0 0 1>}
  n7 = FunctionInput(Preview=TextureCoordinate()) {bUsePreviewValueAsDefault=true, inputName=UV Coord, inputType=FunctionInput_Vector2, previewValue=<512 0 0 1>}
  FunctionOutput(None=Add(A=Multiply(A=@n5, B=@n6), B=@n7))
 ]
}
