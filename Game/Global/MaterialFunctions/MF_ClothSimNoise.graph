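# Material function that produces scrolling, wind-driven noise from the VT_ClothSimWPO texture, apparently for world position offset cloth motion.
# The inverted, normalized wind vector times speed-scaled time offsets the sample position, which is divided by a tiling divisor.
# A static switch selects world-space or actor-relative coordinates, the latter also adding a local velocity vector.
# The sampled RGB is biased, then scaled for output.
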MaterialFunction /Game/Global/MaterialFunctions/MF_ClothSimNoise
graph {nodes [
  n1 = StaticBool()
  n2 = TextureSample(UVs=StaticSwitch(
    True=NamedRerouteUsage()
    False=NamedRerouteUsage()
    Value=FunctionInput(Preview=@n1) {bUsePreviewValueAsDefault=true, inputName=UseLocalUVs, inputType=FunctionInput_StaticBool}
   )) {samplerType=SAMPLERTYPE_LinearColor, texture=&/VT_ClothSimWPO}
  n3 = FunctionInput() {bUsePreviewValueAsDefault=true, inputName=NoiseBias, inputType=FunctionInput_Scalar}
  n4 = FunctionInput() {bUsePreviewValueAsDefault=true, inputName=NoiseScale, inputType=FunctionInput_Scalar, previewValue=<1 0 0 1>}
  FunctionOutput(None=Multiply(A=Add(A=@n2.RGB, B=@n3), B=@n4))
  n5 = Multiply(A=Normalize(VectorInput=FunctionInput() {bUsePreviewValueAsDefault=true, inputName=WindVector, previewValue=<1 0 0 1>})) {constB=-1}
  n6 = FunctionInput() {bUsePreviewValueAsDefault=true, inputName=WindSpeed, inputType=FunctionInput_Scalar}
  NamedRerouteDeclaration(None=Multiply(
    A=NamedRerouteDeclaration(None=@n5) {name=WindVector, nodeColor=<rgba 0.3828 0 1 1>}
    B=MaterialFunctionCall(Speed=@n6) {materialFunction=&TimeWithSpeedVariable}.Time
   )) {name=Wind, nodeColor=<rgba 0 1 0.0859 1>}
  n7 = FunctionInput() {bUsePreviewValueAsDefault=true, inputName=NoiseTileDivisor, inputType=FunctionInput_Scalar, previewValue=<128 0 0 1>}
  NamedRerouteDeclaration(None=@n7) {name=NoiseTileDivisor, nodeColor=<rgba 0.6641 0 1 1>}
  n8 = Subtract(A=WorldPosition().XYZ, B=ActorPositionWS())
  n9 = Add(A=FunctionInput() {bUsePreviewValueAsDefault=true, inputName=LocalVelocityVector}, B=NamedRerouteUsage())
  NamedRerouteDeclaration(None=Divide(A=Add(A=@n8, B=@n9), B=NamedRerouteUsage())) {name=LocalUVs, nodeColor=<rgba 1 0 0.4922 1>}
  n10 = NamedRerouteUsage()
  n11 = WorldPosition()
  NamedRerouteDeclaration(None=Divide(A=Add(A=@n10, B=@n11.XYZ), B=NamedRerouteUsage())) {name=WorldUVs, nodeColor=<rgba 0 0.5312 1 1>}
 ]}
