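# Material function that fakes cloth motion for a region through World Position Offset, driven by velocity and wind.
# Inverted velocity and scaled wind, normalized against a maximum velocity, shape drag and noise-driven turbulence, with optional direction blends toward a custom vector or vertex normal.
# A static switch selects optional collision, which uses projection onto a bone axis, radius and threshold to constrain the offset.
# A second output supplies a collision debug color.
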MaterialFunction /Game/Global/MaterialFunctions/MF_ClothSimWPORegion
graph {nodes [
  n1 = Constant3Vector() {constant=<rgba 0 1 0 0>}
  n2 = Constant3Vector() {constant=<rgba 1 0 0 0>}
  n3 = SmoothStep(Value=MaterialFunctionCall(Vector 3=NamedRerouteUsage()) {materialFunction=/Engine/Functions/Engine_MaterialFunctions02/Utility/VectorLength}.V3 Length) {constMax=2}
  n4 = Multiply(A=NamedRerouteUsage(), B=@n3)
  FunctionOutput(None=NamedRerouteDeclaration(None=LinearInterpolate(A=@n1, B=@n2, Alpha=@n4)) {name=CollisionDetection_DebugColor, nodeColor=<rgba 0.8516 1 0 1>}) {outputName=DebugCollision, sortPriority=100}
  n5 = StaticBool()
  FunctionOutput(None=StaticSwitch(
    True=NamedRerouteUsage()
    False=NamedRerouteUsage()
    Value=FunctionInput(Preview=@n5) {bUsePreviewValueAsDefault=true, inputName=UseCollisionDetection, inputType=FunctionInput_StaticBool, sortPriority=500}
   )) {outputName=WPO}
  n6 = FunctionInput() {bUsePreviewValueAsDefault=true, inputName=WindDragMultiplier, inputType=FunctionInput_Scalar, previewValue=<1 0 0 1>}
  n7 = Add(
   A=NamedRerouteUsage()
   B=Multiply(A=NamedRerouteUsage(), B=NamedRerouteDeclaration(None=@n6) {name=WindDragMultiplier, nodeColor=<rgba 0.2891 0 1 1>})
  )
  MaterialFunctionCall(Vector=@n7, Default=Constant3Vector() {constant=<rgba 0 1 0 0>}) {materialFunction=&SafeNormalize}
  n8 = Add(A=NamedRerouteUsage(), B=NamedRerouteUsage())
  MaterialFunctionCall(Vector=@n8, Default=Constant3Vector() {constant=<rgba 0 1 0 0>}) {materialFunction=&SafeNormalize}
  n9 = NamedRerouteUsage()
  n10 = NamedRerouteUsage()
  n11 = NamedRerouteUsage()
  n12 = NamedRerouteUsage()
  n13 = Normalize(VectorInput=FunctionInput() {bUsePreviewValueAsDefault=true, inputName=TurbulenceVector, previewValue=<0 0 1 1>})
  n14 = Saturate(None=FunctionInput() {bUsePreviewValueAsDefault=true, inputName=UseTurbulenceVector, inputType=FunctionInput_Scalar})
  n15 = FunctionInput(Preview=Constant()) {bUsePreviewValueAsDefault=true, inputName=VertexNormalNoiseInfluence, inputType=FunctionInput_Scalar, sortPriority=200}
  n16 = MaterialFunctionCall(
   Vector=LinearInterpolate(A=LinearInterpolate(A=@n12, B=@n13, Alpha=@n14), B=VertexNormalWS(), Alpha=Saturate(None=@n15))
   Default=Constant3Vector() {constant=<rgba 1 0 0 0>}
  ) {materialFunction=&SafeNormalize}
  NamedRerouteDeclaration(None=Add(A=Multiply(A=@n9, B=@n10), B=Multiply(A=@n11, B=@n16.Result))) {name=WPO, nodeColor=<rgba 0.3828 1 0 1>}
  n17 = FunctionInput() {bUsePreviewValueAsDefault=true, inputName=BoneVector, sortPriority=-49, previewValue=<0 0 1 1>}
  NamedRerouteDeclaration(None=MaterialFunctionCall(Vector=@n17, Default=Constant3Vector() {constant=<rgba 0 0 -1 0>}) {materialFunction=&SafeNormalize}.Result) {name=BoneVector, nodeColor=<rgba 1 0.3047 0 1>}
  n18 = FunctionInput() {bUsePreviewValueAsDefault=true, inputName=BoneRadius, inputType=FunctionInput_Scalar, sortPriority=-48, previewValue=<5 0 0 1>}
  n19 = NamedRerouteDeclaration(None=@n18) {name=BoneRadius, nodeColor=<rgba 0.6406 0 1 1>}
  n20 = FunctionInput() {bUsePreviewValueAsDefault=true, inputName=BoneThreshold, inputType=FunctionInput_Scalar, sortPriority=-47, previewValue=<2 0 0 1>}
  n21 = Add(A=@n19, B=@n20)
  n22 = NamedRerouteUsage()
  NamedRerouteDeclaration(None=Saturate(None=InverseLinearInterpolate(A=@n19, B=@n21, Value=@n22))) {name=ConstraintBlend, nodeColor=<rgba 1 0.0469 0 1>}
  n23 = MaterialFunctionCall(Vector=NamedRerouteUsage(), Default=Constant3Vector() {constant=<rgba 1 0 0 0>}) {materialFunction=&SafeNormalize}
  NamedRerouteDeclaration(None=DotProduct(A=NamedRerouteUsage(), B=NamedRerouteDeclaration(None=@n23.Result) {name=NormalizedOffset, nodeColor=<rgba 0 0.7187 1 1>})) {name=OffsetBoneRadiusProject, nodeColor=<rgba 0 1 0.6719 1>}
  n24 = NamedRerouteUsage()
  n25 = MaterialFunctionCall(Vector 3=@n24) {materialFunction=&VectorLength}
  n26 = NamedRerouteUsage()
  n27 = Constant3Vector() {constant=<rgba 0 0 -1 0>}
  NamedRerouteDeclaration(None=Multiply(
    A=Subtract(A=@n25.V3 Length, B=@n26)
    B=MaterialFunctionCall(Vector=@n24, Default=@n27) {materialFunction=&SafeNormalize}.Result
   )) {name=V2B_short, nodeColor=<rgba 0.1016 1 0 1>}
  NamedRerouteDeclaration(None=MaterialFunctionCall(Vector 3=NamedRerouteUsage()) {materialFunction=&VectorLength}.V3 Length) {name=VelocityMagnitude, nodeColor=<rgba 0.2891 0 1 1>}
  NamedRerouteDeclaration(None=MaterialFunctionCall(Vector 3=NamedRerouteUsage()) {materialFunction=&VectorLength}.V3 Length) {name=WindMagnitude, nodeColor=<rgba 0 1 0.7188 1>}
  NamedRerouteDeclaration(None=MaterialFunctionCall(Vector=NamedRerouteUsage(), Default=Constant3Vector() {constant=<rgba 0 1 0 0>}) {materialFunction=&SafeNormalize}.Result) {name=WindVector, nodeColor=<rgba 1 0 0.4219 1>}
  NamedRerouteDeclaration(None=MaterialFunctionCall(Vector=NamedRerouteUsage(), Default=Constant3Vector() {constant=<rgba 0 1 0 0>}) {materialFunction=&SafeNormalize}.Result) {name=VelocityVector, nodeColor=<rgba 1 0.7969 0 1>}
  n28 = FunctionInput() {bUsePreviewValueAsDefault=true, inputName=Velocity, sortPriority=-200, previewValue=<1 0 0 1>}
  NamedRerouteDeclaration(None=Multiply(A=@n28) {constB=-1}) {name=V, nodeColor=<rgba 1 0.4219 0 1>}
  n29 = FunctionInput() {bUsePreviewValueAsDefault=true, inputName=WindVector, sortPriority=-99, previewValue=<1 0 0 1>}
  n30 = NamedRerouteDeclaration(None=FunctionInput() {bUsePreviewValueAsDefault=true, inputName=WindStrength, inputType=FunctionInput_Scalar, sortPriority=-98}) {name=WindStrength, nodeColor=<rgba 0 1 0.7891 1>}
  n31 = FunctionInput() {bUsePreviewValueAsDefault=true, inputName=MaximumVelocity, inputType=FunctionInput_Scalar, sortPriority=-190, previewValue=<500 0 0 1>}
  n32 = NamedRerouteDeclaration(None=Max(A=@n31)) {name=VelocityMax, nodeColor=<rgba 0.3359 1 0 1>}
  NamedRerouteDeclaration(None=Multiply(A=@n29, B=Multiply(A=@n30, B=@n32))) {name=W, nodeColor=<rgba 0.6641 0 1 1>}
  n33 = NamedRerouteUsage()
  n34 = NamedRerouteUsage()
  NamedRerouteDeclaration(None=Saturate(None=Divide(A=@n33, B=@n34))) {name=VelocityMagnitudeNormalized, nodeColor=<rgba 0.9688 0 1 1>}
  n35 = NamedRerouteUsage()
  n36 = NamedRerouteUsage()
  NamedRerouteDeclaration(None=Saturate(None=Divide(A=@n35, B=@n36))) {name=WindMagnitudeNormalized, nodeColor=<rgba 0.875 1 0 1>}
  NamedRerouteDeclaration(None=Multiply(A=NamedRerouteUsage(), B=NamedRerouteUsage())) {name=FV, nodeColor=<rgba 1 0.5859 0 1>}
  n37 = LinearInterpolate(
   A=NamedRerouteUsage()
   B=MaterialFunctionCall(Vector 3=NamedRerouteUsage()) {materialFunction=&VectorLength}.V3 Length
   Alpha=NamedRerouteUsage()
  )
  n38 = NamedRerouteUsage()
  NamedRerouteDeclaration(None=Multiply(A=@n3, B=Multiply(A=@n37, B=@n38))) {name=WPO_Collision, nodeColor=<rgba 0 0.3438 1 1>}
  NamedRerouteDeclaration(None=Multiply(A=NamedRerouteUsage(), B=NamedRerouteUsage())) {name=WV, nodeColor=<rgba 1 0.9141 0 1>}
  NamedRerouteDeclaration(None=@n7) {name=DragVector, nodeColor=<rgba 0.5703 0 1 1>}
  NamedRerouteDeclaration(None=@n8) {name=TurbulenceVector, nodeColor=<rgba 1 0 0.9375 1>}
  n39 = Max(A=NamedRerouteUsage(), B=NamedRerouteUsage())
  n40 = NamedRerouteUsage()
  NamedRerouteDeclaration(None=Multiply(A=Multiply(A=@n39, B=@n40), B=NamedRerouteUsage())) {name=TurbulenceAmount, nodeColor=<rgba 1 0.5859 0 1>}
  n41 = NamedRerouteUsage()
  n42 = Multiply(A=NamedRerouteUsage(), B=NamedRerouteUsage())
  NamedRerouteDeclaration(None=Multiply(A=NamedRerouteUsage(), B=Max(A=@n41, B=@n42))) {name=DragAmount, nodeColor=<rgba 0 0.0391 1 1>}
  NamedRerouteDeclaration(None=ConstantBiasScale(None=FunctionInput() {bUsePreviewValueAsDefault=true, inputName=TurbulenceNoise, inputType=FunctionInput_Scalar, sortPriority=-78}) {bias=-0.5, scale=2}) {name=TurbulenceNoise, nodeColor=<rgba 0.8984 0 1 1>}
  NamedRerouteDeclaration(None=FunctionInput() {bUsePreviewValueAsDefault=true, inputName=TurbulenceStrength, inputType=FunctionInput_Scalar, sortPriority=-77}) {name=TurbulenceStrength, nodeColor=<rgba 0 0.6016 1 1>}
  n43 = FunctionInput() {bUsePreviewValueAsDefault=true, inputName=DragStrength, inputType=FunctionInput_Scalar, sortPriority=-79, previewValue=<5 0 0 1>}
  NamedRerouteDeclaration(None=@n43) {name=DragStrength, nodeColor=<rgba 0 1 0.2266 1>}
  n44 = FunctionInput() {inputName=BonePosition, sortPriority=-50}
  NamedRerouteDeclaration(None=Subtract(
    A=WorldPosition() {worldPositionShaderOffset=WPT_ExcludeAllShaderOffsets}.XYZ
    B=NamedRerouteDeclaration(None=@n44) {name=BonePosition, nodeColor=<rgba 0.3828 0 1 1>}
   )) {name=BoneToVert, nodeColor=<rgba 0.2422 1 0 1>}
  n45 = WorldPosition() {worldPositionShaderOffset=WPT_ExcludeAllShaderOffsets}
  n46 = NamedRerouteUsage()
  NamedRerouteDeclaration(None=Subtract(A=Add(A=@n45.XYZ, B=@n46), B=NamedRerouteUsage())) {name=BoneToOffsetVert, nodeColor=<rgba 0.7109 0 1 1>}
  n47 = NamedRerouteUsage()
  n48 = DotProduct(A=@n47, B=NamedRerouteUsage())
  NamedRerouteDeclaration(None=Subtract(A=Multiply(A=@n47, B=@n48), B=NamedRerouteUsage())) {name=V2B, nodeColor=<rgba 0 1 0.9297 1>}
  n49 = NamedRerouteUsage()
  n50 = NamedRerouteUsage()
  n51 = Multiply(A=NamedRerouteUsage(), B=NamedRerouteDeclaration(None=DotProduct(A=@n49, B=@n50)) {name=BoneVecProject, nodeColor=<rgba 1 0 0.6797 1>})
  n52 = NamedRerouteUsage()
  NamedRerouteDeclaration(None=MaterialFunctionCall(Vector 3=Subtract(A=@n51, B=@n52)) {materialFunction=&VectorLength}.V3 Length) {name=BoneToOffsetVertLength, nodeColor=<rgba 1 0.8672 0 1>}
  Normalize(VectorInput=@n17)
 ]}
