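# Material function that fakes cloth simulation through World Position Offset for Skin FX materials, behind the UseClothSimWPO switch.
# Vertex color channels mask up to four regions, whose tuning, velocity and bone parameters (named ABP) merge into one MF_ClothSimWPORegion call.
# That call combines wind, drag, MF_ClothSimNoise turbulence and optional bone collision into an offset added to the incoming WPO.
# Static switches can replace shading with debug views.
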MaterialFunction /Game/Global/MaterialFunctions/MF_ClothSimWPO_v2
graph {nodes [
  n1 = NamedRerouteUsage()
  n2 = Normalize(VectorInput=MaterialFunctionCall(Vector=@n1, Default=Constant3Vector() {constant=<rgba 1 0 0 0>}) {materialFunction=&SafeNormalize}.Result)
  ConstantBiasScale(None=@n2)
  FunctionInput() {bUsePreviewValueAsDefault=true, inputName=SkinFX_Mask, inputType=FunctionInput_Vector4, sortPriority=1, previewValue=<1 0 0 1>}
  n3 = GetMaterialAttributes(None=NamedRerouteUsage()) {attributeGetTypes [#1]}
  n4 = NamedRerouteUsage()
  n5 = NamedRerouteUsage()
  n6 = SetMaterialAttributes!(MaterialAttributes=@n3.MaterialAttributes, WorldPositionOffset=Add(A=@n3.WorldPositionOffset, B=Multiply(A=@n4, B=@n5))) {attributeSetTypes [#1]}
  n7 = Constant()
  n8 = SetMaterialAttributes!(
   MaterialAttributes=@n6
   BaseColor=NamedRerouteUsage()
   Metallic=@n7
   Specular=@n7
   Roughness=Constant() {r=1}
   EmissiveColor=Constant3Vector()
  ) {attributeSetTypes [
    #2
    #3
    #4
    #5
    #6
   ]}
  FunctionOutput(None=StaticSwitchParameter(
    True=StaticSwitchParameter(True=@n8, False=@n6) {group=ClothSimWPO, parameterName=ClothSim_ShowDebug, sortPriority=600}
    False=NamedRerouteUsage()
   ) {group=Skin FX, parameterName=UseClothSimWPO})
  Multiply(
   A=Step(X=MaterialFunctionCall(Vector 3=@n1) {materialFunction=/Engine/Functions/Engine_MaterialFunctions02/Utility/VectorLength}.V3 Length) {constY=10}
   B=Abs(None=@n2)
  )
  NamedRerouteDeclaration(None=ScalarParameter() {defaultValue=750, group=ClothSimWPO, parameterName=ClothSim_VelocityMax, sortPriority=0}) {name=VelocityMax, nodeColor=<rgba 0 1 0.4141 1>}
  NamedRerouteDeclaration(None=Normalize(VectorInput=VectorParameter() {group=ClothSimWPO, parameterName=ClothSim_WindVector, sortPriority=1, defaultValue=<rgba 0 1 0 1>}.RGB)) {name=WindVector, nodeColor=<rgba 0 1 0.3203 1>}
  n9 = Multiply(A=NamedRerouteUsage(), B=NamedRerouteUsage())
  n10 = NamedRerouteUsage()
  n11 = NamedRerouteUsage()
  n12 = NamedRerouteUsage()
  n13 = NamedRerouteUsage()
  n14 = NamedRerouteUsage()
  n15 = NamedRerouteUsage()
  n16 = StaticSwitchParameter(True=Add(A=Multiply(A=@n12, B=@n13), B=Multiply(A=@n14, B=@n15)), False=NamedRerouteUsage()) {group=ClothSimWPO, parameterName=UseRegion02, sortPriority=200}
  n17 = StaticSwitchParameter(True=Add(A=Multiply(A=@n10, B=@n11), B=@n16), False=@n16) {group=ClothSimWPO, parameterName=UseRegion03, sortPriority=300}
  NamedRerouteDeclaration(None=StaticSwitchParameter(True=Add(A=@n9, B=@n17), False=@n17) {group=ClothSimWPO, parameterName=UseRegion04, sortPriority=400}) {name=Params, nodeColor=<rgba 1 0 0.4922 1>}
  n18 = VectorParameter() {group=ClothSimWPO, parameterName=ClothSim_ABP_R1_Velocity, sortPriority=150, defaultValue=<rgba 0 0 0 1>}
  NamedRerouteDeclaration(None=@n18.RGB) {name=R1_Velocity, nodeColor=<rgba 0.5 1 0 1>}
  NamedRerouteDeclaration(None=ScalarParameter() {defaultValue=128, group=ClothSimWPO, parameterName=ClothSim_TurbulenceNoiseTileDivisor, sortPriority=4}) {name=TurbulenceNoiseTileDivisor, nodeColor=<rgba 0.5938 0 1 1>}
  n19 = VectorParameter() {group=ClothSimWPO, parameterName=ClothSim_ABP_R2_Velocity, sortPriority=250, defaultValue=<rgba 0 0 0 1>}
  NamedRerouteDeclaration(None=@n19.RGB) {name=R2_Velocity, nodeColor=<rgba 0 1 0.2266 1>}
  n20 = VectorParameter() {group=ClothSimWPO, parameterName=ClothSim_ABP_R3_Velocity, sortPriority=50, defaultValue=<rgba 0 0 0 1>}
  NamedRerouteDeclaration(None=@n20.RGB) {name=R3_Velocity, nodeColor=<rgba 0.3594 0 1 1>}
  n21 = VertexColor()
  n22 = AppendVector(A=@n21, B=@n21)
  n23 = ChannelMaskParameter(None=@n22) {group=ClothSimWPO, maskChannel=Blue, parameterName=ClothSim_Region_03_MaskChannel, sortPriority=310, defaultValue=<rgba 0 0 1 0>}
  NamedRerouteDeclaration(None=@n23.R) {name=R3_MaskChannel, nodeColor=<rgba 0.2188 0 1 1>}
  n24 = VectorParameter() {group=ClothSimWPO, parameterName=ClothSim_ABP_R4_Velocity, sortPriority=450, defaultValue=<rgba 0 0 0 1>}
  NamedRerouteDeclaration(None=@n24.RGB) {name=R4_Velocity, nodeColor=<rgba 0.1484 1 0 1>}
  n25 = ChannelMaskParameter(None=@n22) {group=ClothSimWPO, maskChannel=Alpha, parameterName=ClothSim_Region_04_MaskChannel, sortPriority=410, defaultValue=<rgba 0 0 0 1>}
  NamedRerouteDeclaration(None=@n25.R) {name=R4_MaskChannel, nodeColor=<rgba 0.2188 0 1 1>}
  NamedRerouteDeclaration(None=ChannelMaskParameter(None=@n22) {group=ClothSimWPO, parameterName=ClothSim_Region_01_MaskChannel, sortPriority=100}.R) {name=R1_MaskChannel, nodeColor=<rgba 1 0 0.3281 1>}
  n26 = VectorParameter() {group=ClothSimWPO, parameterName=ClothSim_ABP_R1_BonePosition, sortPriority=160, defaultValue=<rgba 0 0 0 1>}
  NamedRerouteDeclaration(None=@n26.RGB) {name=R1_BonePosition, nodeColor=<rgba 1 0 0.1641 1>}
  n27 = VectorParameter() {group=ClothSimWPO, parameterName=ClothSim_ABP_R2_BonePosition, sortPriority=260, defaultValue=<rgba 0 0 0 1>}
  NamedRerouteDeclaration(None=@n27.RGB) {name=R2_BonePosition, nodeColor=<rgba 0 0.2031 1 1>}
  n28 = VectorParameter() {group=ClothSimWPO, parameterName=ClothSim_ABP_R3_BonePosition, sortPriority=360, defaultValue=<rgba 0 0 0 1>}
  NamedRerouteDeclaration(None=@n28.RGB) {name=R3_BonePosition, nodeColor=<rgba 0.3828 0 1 1>}
  n29 = VectorParameter() {group=ClothSimWPO, parameterName=ClothSim_ABP_R4_BonePosition, sortPriority=460, defaultValue=<rgba 0 0 0 1>}
  NamedRerouteDeclaration(None=@n29.RGB) {name=R4_BonePosition, nodeColor=<rgba 0.7578 1 0 1>}
  n30 = VectorParameter() {group=ClothSimWPO, parameterName=ClothSim_ABP_R1_BoneVector, sortPriority=161, defaultValue=<rgba 0 0 0 1>}
  NamedRerouteDeclaration(None=@n30.RGB) {name=R1_BoneVector, nodeColor=<rgba 0.8984 1 0 1>}
  n31 = VectorParameter() {group=ClothSimWPO, parameterName=ClothSim_ABP_R2_BoneVector, sortPriority=261, defaultValue=<rgba 0 0 0 1>}
  NamedRerouteDeclaration(None=@n31.RGB) {name=R2_BoneVector, nodeColor=<rgba 0.8984 1 0 1>}
  n32 = VectorParameter() {group=ClothSimWPO, parameterName=ClothSim_ABP_R3_BoneVector, sortPriority=361, defaultValue=<rgba 0 0 0 1>}
  NamedRerouteDeclaration(None=@n32.RGB) {name=R3_BoneVector, nodeColor=<rgba 0.8984 1 0 1>}
  n33 = VectorParameter() {group=ClothSimWPO, parameterName=ClothSim_ABP_R4_BoneVector, sortPriority=461, defaultValue=<rgba 0 0 0 1>}
  NamedRerouteDeclaration(None=@n33.RGB) {name=R4_BoneVector, nodeColor=<rgba 0.8984 1 0 1>}
  NamedRerouteDeclaration(None=MaterialFunctionCall(
    R=ScalarParameter() {defaultValue=5, group=ClothSimWPO, parameterName=ClothSim_R1_DragStrength, sortPriority=120}
    G=ScalarParameter() {defaultValue=3, group=ClothSimWPO, parameterName=ClothSim_R1_TurbulenceStrength, sortPriority=121}
    B=ScalarParameter() {defaultValue=5, group=ClothSimWPO, parameterName=ClothSim_R1_BoneRadius, sortPriority=122}
    A=ScalarParameter() {defaultValue=2, group=ClothSimWPO, parameterName=ClothSim_R1_BoneThreshold, sortPriority=123}
   ) {materialFunction=&AppendMany}.RGBA) {name=R1_Params, nodeColor=<rgba 1 0.3047 0 1>}
  NamedRerouteDeclaration(None=MaterialFunctionCall(
    R=ScalarParameter() {defaultValue=4, group=ClothSimWPO, parameterName=ClothSim_R2_DragStrength, sortPriority=220}
    G=ScalarParameter() {defaultValue=3, group=ClothSimWPO, parameterName=ClothSim_R2_TurbulenceStrength, sortPriority=221}
    B=ScalarParameter() {defaultValue=5, group=ClothSimWPO, parameterName=ClothSim_R2_BoneRadius, sortPriority=222}
    A=ScalarParameter() {defaultValue=2, group=ClothSimWPO, parameterName=ClothSim_R2_BoneThreshold, sortPriority=223}
   ) {materialFunction=&AppendMany}.RGBA) {name=R2_Params, nodeColor=<rgba 0.2656 0 1 1>}
  n34 = ChannelMaskParameter(None=@n22) {group=ClothSimWPO, maskChannel=Green, parameterName=ClothSim_Region_02_MaskChannel, sortPriority=210, defaultValue=<rgba 0 1 0 0>}
  NamedRerouteDeclaration(None=@n34.R) {name=R2_MaskChannel, nodeColor=<rgba 1 0 0.3984 1>}
  NamedRerouteDeclaration(None=MaterialFunctionCall(
    R=ScalarParameter() {defaultValue=4, group=ClothSimWPO, parameterName=ClothSim_R3_DragStrength, sortPriority=320}
    G=ScalarParameter() {defaultValue=3, group=ClothSimWPO, parameterName=ClothSim_R3_TurbulenceStrength, sortPriority=321}
    B=ScalarParameter() {defaultValue=5, group=ClothSimWPO, parameterName=ClothSim_R3_BoneRadius, sortPriority=322}
    A=ScalarParameter() {defaultValue=2, group=ClothSimWPO, parameterName=ClothSim_R3_BoneThreshold, sortPriority=323}
   ) {materialFunction=&AppendMany}.RGBA) {name=R3_Params, nodeColor=<rgba 0.2656 0 1 1>}
  NamedRerouteDeclaration(None=MaterialFunctionCall(
    R=ScalarParameter() {defaultValue=4, group=ClothSimWPO, parameterName=ClothSim_R4_DragStrength, sortPriority=420}
    G=ScalarParameter() {defaultValue=3, group=ClothSimWPO, parameterName=ClothSim_R4_TurbulenceStrength, sortPriority=421}
    B=ScalarParameter() {defaultValue=5, group=ClothSimWPO, parameterName=ClothSim_R4_BoneRadius, sortPriority=422}
    A=ScalarParameter() {defaultValue=2, group=ClothSimWPO, parameterName=ClothSim_R4_BoneThreshold, sortPriority=423}
   ) {materialFunction=&AppendMany}.RGBA) {name=R4_Params, nodeColor=<rgba 0.2656 0 1 1>}
  n35 = NamedRerouteUsage()
  n36 = StaticSwitchParameter(True=Add(A=NamedRerouteUsage(), B=NamedRerouteUsage()), False=NamedRerouteUsage()) {group=ClothSimWPO, parameterName=UseRegion02, sortPriority=200}
  n37 = StaticSwitchParameter(True=Add(A=NamedRerouteUsage(), B=@n36), False=@n36) {group=ClothSimWPO, parameterName=UseRegion03, sortPriority=300}
  NamedRerouteDeclaration(None=StaticSwitchParameter(True=Add(A=@n35, B=@n37), False=@n37) {group=ClothSimWPO, parameterName=UseRegion04, sortPriority=400}) {name=Mask, nodeColor=<rgba 0 0.7891 1 1>}
  NamedRerouteDeclaration(None=FunctionInput() {inputName=Material, inputType=FunctionInput_MaterialAttributes}) {name=Material, nodeColor=<rgba 1 0.7969 0 1>}
  NamedRerouteDeclaration(None=VectorParameter() {group=ClothSimWPO, parameterName=ClothSim_ABP_TurbulenceUVOffset, defaultValue=<rgba 0 0 0 1>}.RGB) {name=TurbulenceUVOffset, nodeColor=<rgba 0.7578 0 1 1>}
  n38 = NamedRerouteUsage()
  n39 = Multiply(A=NamedRerouteUsage(), B=NamedRerouteDeclaration(None=Step(X=@n38) {constY=0.001}) {name=R4_BitMask, nodeColor=<rgba 0 1 0.9766 1>})
  n40 = NamedRerouteUsage()
  n41 = NamedRerouteDeclaration(None=Step(X=NamedRerouteUsage()) {constY=0.001}) {name=R3_BitMask, nodeColor=<rgba 0 1 0.9766 1>}
  n42 = NamedRerouteUsage()
  n43 = NamedRerouteDeclaration(None=Step(X=NamedRerouteUsage()) {constY=0.001}) {name=R1_BitMask, nodeColor=<rgba 1 0 0.7266 1>}
  n44 = NamedRerouteUsage()
  n45 = NamedRerouteDeclaration(None=Step(X=NamedRerouteUsage()) {constY=0.001}) {name=R2_BitMask, nodeColor=<rgba 1 0 0.7266 1>}
  n46 = StaticSwitchParameter(True=Add(A=Multiply(A=@n42, B=@n43), B=Multiply(A=@n44, B=@n45)), False=NamedRerouteUsage()) {group=ClothSimWPO, parameterName=UseRegion02, sortPriority=200}
  n47 = StaticSwitchParameter(True=Add(A=Multiply(A=@n40, B=@n41), B=@n46), False=@n46) {group=ClothSimWPO, parameterName=UseRegion03, sortPriority=300}
  NamedRerouteDeclaration(None=StaticSwitchParameter(True=Add(A=@n39, B=@n47), False=@n47) {group=ClothSimWPO, parameterName=UseRegion04, sortPriority=400}) {name=Velocity, nodeColor=<rgba 0 0.7187 1 1>}
  NamedRerouteDeclaration(None=ScalarParameter() {defaultValue=1.5, group=ClothSimWPO, parameterName=ClothSim_WindNoiseSpeed, sortPriority=3}) {name=WindSpeed, nodeColor=<rgba 0.5234 1 0 1>}
  NamedRerouteDeclaration(None=MaterialFunctionCall!(
    WindSpeed=NamedRerouteUsage()
    NoiseTileDivisor=NamedRerouteUsage()
    LocalVelocityVector=NamedRerouteUsage()
    UseLocalUVs=StaticBoolParameter() {group=ClothSimWPO, parameterName=ClothSim_UseNoiseLocalUVs, sortPriority=4}
    NoiseBias=ScalarParameter() {group=ClothSimWPO, parameterName=ClothSim_TurbulenceBias, sortPriority=5}
    NoiseScale=ScalarParameter() {defaultValue=1, group=ClothSimWPO, parameterName=ClothSim_TurbulenceScale, sortPriority=5}
    WindVector=NamedRerouteUsage()
   ) {materialFunction=&/MF_ClothSimNoise}.Result) {name=TurbulenceNoise, nodeColor=<rgba 0 0.5547 1 1>}
  NamedRerouteDeclaration(None=ScalarParameter() {group=ClothSimWPO, parameterName=ClothSim_WindStrength, sortPriority=2}) {name=WindStrength, nodeColor=<rgba 0.4766 1 0 1>}
  n48 = NamedRerouteUsage()
  n49 = NamedRerouteUsage()
  n50 = Constant3Vector()
  n51 = StaticSwitchParameter(
   True=NamedRerouteUsage()
   False=StaticSwitchParameter(
    True=NamedRerouteUsage()
    False=StaticSwitchParameter(True=@n49, False=@n50) {defaultValue=true, group=ClothSimWPO, parameterName=ShowDebugNoise, sortPriority=601}
   ) {group=ClothSimWPO, parameterName=ShowDebugVertexColor, sortPriority=602}
  ) {group=ClothSimWPO, parameterName=ShowDebugVelocity, sortPriority=603}
  NamedRerouteDeclaration(None=StaticSwitchParameter(
    True=NamedRerouteUsage()
    False=StaticSwitchParameter(True=@n48, False=@n51) {group=ClothSimWPO, parameterName=ShowDebugVelocityMagnitude, sortPriority=604}
   ) {group=ClothSimWPO, parameterName=ShowDebugCollision, sortPriority=605}) {name=DebugColor, nodeColor=<rgba 0.8281 0 1 1>}
  n52 = Multiply(A=NamedRerouteUsage()) {constB=-1}
  n53 = CrossProduct(
   A=MaterialFunctionCall(Vector=Normalize(VectorInput=@n52), Default=Constant3Vector() {constant=<rgba 0 0 1 0>}) {materialFunction=&SafeNormalize}.Result
   B=Constant3Vector() {constant=<rgba 1 0 0 0>}
  )
  NamedRerouteDeclaration(None=Abs(None=Normalize(VectorInput=@n53))) {name=DebugVelocity, nodeColor=<rgba 1 0.8906 0 1>}
  NamedRerouteDeclaration(None=NamedRerouteUsage()) {name=DebugNoise, nodeColor=<rgba 1 0.7031 0 1>}
  n54 = MaterialFunctionCall(Vector 3=NamedRerouteUsage()) {materialFunction=&VectorLength}
  n55 = NamedRerouteUsage()
  NamedRerouteDeclaration(None=Saturate(None=Divide(A=@n54.V3 Length, B=@n55))) {name=DebugVelocityMagnitude, nodeColor=<rgba 1 0 0.6797 1>}
  n56 = VertexColor()
  n57 = Constant3Vector() {constant=<rgba 1 0 1 1>}
  NamedRerouteDeclaration(None=Add(A=@n56, B=Multiply(A=@n56.A, B=@n57))) {name=DebugVertexColor, nodeColor=<rgba 0.9453 1 0 1>}
  n58 = MaterialFunctionCall!(
   Velocity=NamedRerouteUsage()
   MaximumVelocity=NamedRerouteUsage()
   WindVector=NamedRerouteUsage()
   WindStrength=NamedRerouteUsage()
   DragStrength=ComponentMask(None=NamedRerouteUsage()) {r=true}
   TurbulenceNoise=NamedRerouteUsage()
   TurbulenceStrength=ComponentMask(None=NamedRerouteUsage()) {g=true}
   BonePosition=NamedRerouteUsage()
   BoneVector=NamedRerouteUsage()
   BoneRadius=ComponentMask(None=NamedRerouteUsage()) {b=true}
   BoneThreshold=ComponentMask(None=NamedRerouteUsage()) {a=true}
   UseTurbulenceVector=ScalarParameter() {group=ClothSimWPO, parameterName=ClothSim_ABP_TurbulenceVector_Strength, sortPriority=6}
   TurbulenceVector=VectorParameter() {group=ClothSimWPO, parameterName=ClothSim_ABP_TurbulenceVector, defaultValue=<rgba 0 0 1 1>}.RGB
   WindDragMultiplier=ScalarParameter() {defaultValue=1, group=ClothSimWPO, parameterName=ClothSim_WindDragMultiplier, sortPriority=4}
   VertexNormalNoiseInfluence=ScalarParameter() {group=ClothSimWPO, parameterName=ClothSim_VertexNormalNoiseInfluence, sortPriority=5}
   UseCollisionDetection=StaticBoolParameter() {group=ClothSimWPO, parameterName=ClothSim_UseCollisionDetection, sortPriority=5}
  ) {materialFunction=&/MF_ClothSimWPORegion}
  NamedRerouteDeclaration(None=@n58.DebugCollision) {name=DebugCollision, nodeColor=<rgba 1 0.6563 0 1>}
  NamedRerouteDeclaration(None=@n58.WPO) {name=WPO, nodeColor=<rgba 0 1 0.1562 1>}
  n59 = Multiply(A=NamedRerouteUsage(), B=NamedRerouteUsage())
  n60 = NamedRerouteUsage()
  n61 = NamedRerouteUsage()
  n62 = NamedRerouteUsage()
  n63 = NamedRerouteUsage()
  n64 = NamedRerouteUsage()
  n65 = NamedRerouteUsage()
  n66 = StaticSwitchParameter(True=Add(A=Multiply(A=@n62, B=@n63), B=Multiply(A=@n64, B=@n65)), False=NamedRerouteUsage()) {group=ClothSimWPO, parameterName=UseRegion02, sortPriority=200}
  n67 = StaticSwitchParameter(True=Add(A=Multiply(A=@n60, B=@n61), B=@n66), False=@n66) {group=ClothSimWPO, parameterName=UseRegion03, sortPriority=300}
  NamedRerouteDeclaration(None=StaticSwitchParameter(True=Add(A=@n59, B=@n67), False=@n67) {group=ClothSimWPO, parameterName=UseRegion04, sortPriority=400}) {name=BonePosition, nodeColor=<rgba 0 0.7188 1 1>}
  n68 = Multiply(A=NamedRerouteUsage(), B=NamedRerouteUsage())
  n69 = NamedRerouteUsage()
  n70 = NamedRerouteUsage()
  n71 = Multiply(A=NamedRerouteUsage(), B=NamedRerouteUsage())
  n72 = Multiply(A=NamedRerouteUsage(), B=NamedRerouteUsage())
  n73 = StaticSwitchParameter(
   True=Add(A=Multiply(A=@n69, B=@n70))
   False=StaticSwitchParameter(True=Add(A=@n71, B=@n72), False=NamedRerouteUsage()) {group=ClothSimWPO, parameterName=UseRegion02, sortPriority=200}
  ) {group=ClothSimWPO, parameterName=UseRegion03, sortPriority=300}
  NamedRerouteDeclaration(None=StaticSwitchParameter(True=Add(A=@n68, B=@n73), False=@n73) {group=ClothSimWPO, parameterName=UseRegion04, sortPriority=400}) {name=BoneVector, nodeColor=<rgba 0 0.7188 1 1>}
 ]}
