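# Material function that adds a simple deploy grow-and-wobble animation through World Position Offset, toggled by the UseDeployWobbleSimple static switch.
# Transition progress comes from a TimestampA custom primitive data value, or from looping modulo time under the DebugTime switch.
# Vertices start pulled toward a biased object pivot and expand with smoothstepped, per-axis staggered timing.
# A time-driven sine wobble along the pivot direction fades out as the transition completes.
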MaterialFunction /Game/Global/MaterialFunctions/MF_DeployWobble_Simple
graph {nodes [
  n1 = FunctionInput() {inputType=FunctionInput_MaterialAttributes}
  n2 = GetMaterialAttributes(None=@n1) {attributeGetTypes [#1]}
  n3 = Subtract(
   A=Add(
    A=MaterialFunctionCall() {materialFunction=&ObjectPivotPoint}.Object Pivot Location
    B=VectorParameter() {group=DeployWobbleSimple, parameterName=PivotBias, defaultValue=<rgba 0 0 -40 0>}.RGB
   )
   B=WorldPosition().XYZ
  )
  n4 = StaticSwitchParameter(
   True=MaterialFunctionCall!(Time, Divisor, Mod) {materialFunction=/Game/Effects/Fort_Effects/Materials/_Functions/MF_ModuloTime}.Clamped 0-1
   False=MaterialFunctionCall(
    TimestampA=ScalarParameter() {bUseCustomPrimitiveData=true, parameterName=TimestampA}
    TransitionLength_In=ScalarParameter() {defaultValue=0.5, group=DeployWobbleSimple, parameterName=TransitionLength_In}
    Transition In Only?=StaticBool() {value=true}
   ) {materialFunction=&MF_Timestamp_InOut_01}.Linear
  ) {parameterName=DebugTime}
  n5 = Multiply(
   A=Add(A=@n4, B=VectorParameter() {group=DeployWobbleSimple, parameterName=GrowthBias, defaultValue=<rgba 0 0 -0.1 0>}.RGB)
   B=VectorParameter() {group=DeployWobbleSimple, parameterName=GrowthScale, defaultValue=<rgba 4 3 2 0>}.RGB
  ) {constB=2}
  n6 = SmoothStep(Value=Saturate(None=OneMinus(None=@n5)))
  n7 = Time()
  n8 = Multiply(
   A=Normalize(VectorInput=@n3)
   B=Multiply(A=Sine(None=@n7) {period=0.18}, B=ScalarParameter() {defaultValue=6, group=DeployWobbleSimple, parameterName=SecondaryWobbleScale})
  )
  n9 = OneMinus(None=@n4)
  n10 = Add(A=@n2.WorldPositionOffset, B=Add(A=Multiply(A=@n3, B=@n6), B=Multiply(A=@n8, B=@n9)))
  FunctionOutput(None=StaticSwitchParameter(
    True=SetMaterialAttributes!(MaterialAttributes=@n2.MaterialAttributes, WorldPositionOffset=@n10) {attributeSetTypes [#1]}
    False=@n1
   ) {group=DeployWobbleSimple, parameterName=UseDeployWobbleSimple})
 ]}
