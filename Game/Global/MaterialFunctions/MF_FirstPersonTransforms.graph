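# Material function that converts a camera-relative position and a scale into first-person view space, blended by FirstPersonLerpAlpha.
# Custom HLSL multiplies the position by View.FirstPersonTransform and lerps from the original to output the adjusted position.
# For scale, it transforms points offset along each local axis by that scale component, measures their distances to the adjusted position, then lerps from the input scale to those lengths.
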MaterialFunction /Game/Global/MaterialFunctions/MF_FirstPersonTransforms
graph {
 properties {bExposeToLibrary=true}
 nodes [
  Constant() {r=1}
  MaterialExpressionNamedRerouteDeclaration_0 = NamedRerouteDeclaration(None=FunctionInput() {inputName=CameraRelativePosition}) {name=CameraRelativePositon, nodeColor=<rgba 0 1 0.4375 1>}
  MaterialExpressionNamedRerouteUsage_7 = NamedRerouteUsage()
  n1 = Custom!(TranslatedWorldPosition=@MaterialExpressionNamedRerouteDeclaration_0, FirstPersonLerpAlpha=@MaterialExpressionNamedRerouteUsage_7) {
   code="const float3 FirstPersonPosition = mul(TranslatedWorldPosition, (float3x3)View.FirstPersonTransform);\r\nreturn lerp(TranslatedWorldPosition, FirstPersonPosition , FirstPersonLerpAlpha);"
   showCode=true
   inputs [
    {
     inputName=TranslatedWorldPosition
     input {expression=&self:MaterialExpressionNamedRerouteDeclaration_0}
    }
    {
     inputName=FirstPersonLerpAlpha
     input {expression=&self:MaterialExpressionNamedRerouteUsage_7}
    }
   ]
  }
  FunctionOutput(None=@n1) {outputName=CameraRealtiveFP}
  MaterialExpressionMultiply_0 = Multiply(
   A=Transform(None=Constant3Vector() {constant=<rgba 1 0 0 0>}) {transformSourceType=TRANSFORMSOURCE_Local}
   B=ComponentMask(None=NamedRerouteUsage()) {r=true}
  )
  MaterialExpressionNamedRerouteUsage_0 = NamedRerouteUsage()
  n2 = Custom!(Radius=@MaterialExpressionMultiply_0, Center=@MaterialExpressionNamedRerouteUsage_0) {
   code="const float3 ScaleFP = mul((Center + Radius), (float3x3)View.FirstPersonTransform);\r\nreturn ScaleFP;"
   showCode=true
   inputs [
    {inputName=Radius, input {expression=&self:MaterialExpressionMultiply_0}}
    {inputName=Center, input {expression=&self:MaterialExpressionNamedRerouteUsage_0}}
   ]
  }
  MaterialExpressionMultiply_1 = Multiply(
   A=Transform(None=Constant3Vector() {constant=<rgba 0 1 0 0>}) {transformSourceType=TRANSFORMSOURCE_Local}
   B=ComponentMask(None=NamedRerouteUsage()) {g=true}
  )
  MaterialExpressionNamedRerouteUsage_1 = NamedRerouteUsage()
  n3 = Custom(Radius=@MaterialExpressionMultiply_1, Center=@MaterialExpressionNamedRerouteUsage_1) {
   code="const float3 ScaleFP = mul((Center + Radius), (float3x3)View.FirstPersonTransform);\r\nreturn ScaleFP;"
   showCode=true
   inputs [
    {inputName=Radius, input {expression=&self:MaterialExpressionMultiply_1}}
    {inputName=Center, input {expression=&self:MaterialExpressionNamedRerouteUsage_1}}
   ]
  }
  n4 = AppendVector(A=Distance(A=@n2, B=@n1), B=Distance(A=@n3, B=@n1))
  MaterialExpressionMultiply_2 = Multiply(
   A=Transform(None=Constant3Vector() {constant=<rgba 0 0 1 0>}) {transformSourceType=TRANSFORMSOURCE_Local}
   B=ComponentMask(None=NamedRerouteUsage()) {b=true}
  )
  MaterialExpressionNamedRerouteUsage_2 = NamedRerouteUsage()
  n5 = Custom(Radius=@MaterialExpressionMultiply_2, Center=@MaterialExpressionNamedRerouteUsage_2) {
   code="const float3 ScaleFP = mul((Center + Radius), (float3x3)View.FirstPersonTransform);\r\nreturn ScaleFP;"
   showCode=true
   inputs [
    {inputName=Radius, input {expression=&self:MaterialExpressionMultiply_2}}
    {inputName=Center, input {expression=&self:MaterialExpressionNamedRerouteUsage_2}}
   ]
  }
  n6 = Distance(A=@n5, B=@n1)
  FunctionOutput(None=LinearInterpolate(A=NamedRerouteUsage(), B=AppendVector(A=@n4, B=@n6), Alpha=NamedRerouteUsage())) {outputName=Scale, sortPriority=55}
  NamedRerouteDeclaration(None=FunctionInput() {inputName=Scale, sortPriority=55}) {name=Scale, nodeColor=<rgba 0.2891 1 0 1>}
  NamedRerouteDeclaration(None=FunctionInput() {inputName=FirstPersonLerpAlpha, inputType=FunctionInput_Scalar, sortPriority=99}) {name=FirstPersonLerpAlpha, nodeColor=<rgba 1 0 0.6094 1>}
 ]
}
