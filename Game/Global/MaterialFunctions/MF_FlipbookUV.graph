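# Material function that computes flipbook UVs and blend weights for crossfading between adjacent frames of a grid atlas.
# A single Custom HLSL node takes the UV, column and row count, and selector, wrapping the selector with frac.
# It maps UVs into the current and next cells and derives both weights from the fractional frame position.
# The next index is not wrapped, so on the last frame its UV lands outside the 0-1 range.
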MaterialFunction /Game/Global/MaterialFunctions/MF_FlipbookUV
graph {nodes [
  MaterialExpressionFunctionInput_2 = FunctionInput() {bUsePreviewValueAsDefault=true, description=Input UV, inputName=UV, inputType=FunctionInput_Vector2, sortPriority=2}
  MaterialExpressionFunctionInput_0 = FunctionInput() {bUsePreviewValueAsDefault=true, description="Flipbook Cols n Rows ", inputName=Size, inputType=FunctionInput_Vector2, previewValue=<1 1 0 1>}
  MaterialExpressionFunctionInput_1 = FunctionInput() {bUsePreviewValueAsDefault=true, description="Frame Selector (0-1)", inputName=Selector, inputType=FunctionInput_Scalar, sortPriority=1}
  n1 = Custom!(InstanceUV=@MaterialExpressionFunctionInput_2, Size=@MaterialExpressionFunctionInput_0, Selector=@MaterialExpressionFunctionInput_1) {
   code="float width = Size.x;\r\nfloat height = Size.y;\r\nfloat area = width * height;\r\nfloat sel = frac(Selector);\r\nfloat findex = area * sel;\r\nfloat index1 = floor(findex);\r\nfloat index2 = index1 + 1;\r\nfloat row1 = floor(index1 / width);\r\nfloat row2 = floor(index2 / width);\r\nfloat col1 = index1 - row1 * width;\r\nfloat col2 = index2 - row2 * width;\r\nfloat2 gridPos1 = float2(col1, row1);\r\nfloat2 gridPos2 = float2(col2, row2);\r\nNewUV1 = (gridPos1 + InstanceUV) / Size;\r\nNewUV2 = (gridPos2 + InstanceUV) / Size;\r\nBlend2 = findex - index1;\r\nBlend1 = 1.0 - Blend2;\r\nreturn Size;"
   description=Flipbook
   outputType=CMOT_Float1
   showCode=true
   additionalOutputs [
    {outputName=NewUV1, outputType=CMOT_Float2}
    {outputName=NewUV2, outputType=CMOT_Float2}
    {outputName=Blend1, outputType=CMOT_Float1}
    {outputName=Blend2, outputType=CMOT_Float1}
   ]
   inputs [
    {inputName=InstanceUV, input {expression=&self:MaterialExpressionFunctionInput_2}}
    {inputName=Size, input {expression=&self:MaterialExpressionFunctionInput_0}}
    {inputName=Selector, input {expression=&self:MaterialExpressionFunctionInput_1}}
   ]
  }
  FunctionOutput(None=@n1.NewUV1) {outputName=FlipbookUV1}
  FunctionOutput(None=@n1.NewUV2) {outputName=FlipbookUV2, sortPriority=1}
  FunctionOutput(None=@n1.Blend1) {outputName=Blend1, sortPriority=2}
  FunctionOutput(None=@n1.Blend2) {outputName=Blend2, sortPriority=3}
 ]}
