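# MF_Gain is a library-exposed material function that applies a gain curve to a scalar value, controlled by a gain input defaulting to 0.5.
# It doubles the value and shapes each half of the range with MF_Bias, using an offset value and inverted bias for the upper half.
# An If at the midpoint selects between the rescaled halves.
# Named reroutes distribute the inputs, with UV red previewing the value.
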MaterialFunction /Game/Global/MaterialFunctions/MF_Gain
graph {
 properties {bExposeToLibrary=true}
 nodes [
  n1 = Multiply(A=NamedRerouteUsage()) {constB=2}
  n2 = Subtract(A=@n1)
  n3 = Subtract(B=NamedRerouteUsage())
  n4 = Add(A=Divide(A=MaterialFunctionCall!(value=@n2, bias=@n3) {materialFunction=&/MF_Bias}.Result)) {constB=0.5}
  n5 = MaterialFunctionCall(value=@n1, bias=NamedRerouteUsage()) {materialFunction=&MF_Bias}
  FunctionOutput(None=If(A=NamedRerouteUsage(), "A > B"=@n4, "A == B"=@n4, "A < B"=Divide(A=@n5.Result)) {constB=0.5})
  n6 = TextureCoordinate()
  NamedRerouteDeclaration(None=FunctionInput(Preview=ComponentMask(None=@n6) {r=true}) {bUsePreviewValueAsDefault=true, inputName=value, inputType=FunctionInput_Scalar}) {name=value, nodeColor=<rgba 1 0 0.2578 1>}
  n7 = FunctionInput() {bUsePreviewValueAsDefault=true, inputName=gain, inputType=FunctionInput_Scalar, sortPriority=1, previewValue=<0.5 0 0 1>}
  NamedRerouteDeclaration(None=@n7) {name=gain, nodeColor=<rgba 1 0.9375 0 1>}
 ]
}
