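# Material function that scales a material's emissive color by a project-wide multiplier.
# It takes Material Attributes, multiplies EmissiveColor by GlobalEmissiveMultiplier from the MPC_GlobalEmissive parameter collection, and writes the result back into the attributes.
# A UseGlobalEmissiveMultiplier static switch selects the scaled or original emissive at compile time, passing other attributes through unchanged.
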MaterialFunction /Game/Global/MaterialFunctions/MF_GlobalEmissiveMultiplier_MPC
graph {nodes [
  n1 = GetMaterialAttributes(None=FunctionInput() {inputType=FunctionInput_MaterialAttributes}) {attributeGetTypes [#1]}
  n2 = Multiply(
   A=@n1.EmissiveColor
   B=CollectionParameter() {collection=/Game/Global/MaterialParameterCollections/MPC_GlobalEmissive, parameterName=GlobalEmissiveMultiplier}
  )
  FunctionOutput(None=SetMaterialAttributes!(
    MaterialAttributes=@n1.MaterialAttributes
    EmissiveColor=StaticSwitchParameter(True=@n2, False=@n1.EmissiveColor) {parameterName=UseGlobalEmissiveMultiplier}
   ) {attributeSetTypes [#1]})
 ]}
