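# Material function that computes cubemap UVWs to fake parallax rooms behind surfaces such as building windows.
# Tiled UVs split into room cells, where a ray-box intersection uses the view vector scaled by RoomZscale.
# UseBoxTangentTransforms switches that vector between tangent space and a box basis blended from local axes by vertex normal.
# Randomize Rotation varies rooms by flipping and swizzling the result according to cell index parity, so it is deterministic.
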MaterialFunction /Game/Global/MaterialFunctions/MF_InteriorCubemapUV
graph {
 properties {
  bExposeToLibrary=true
  description=Interior Cubemap conforms a cubemap to be in the form of a box shape inset from the surface UVs. It is useful for giving parallax to rooms visible through windows on city buildings.
  libraryCategoriesText [Cubemaps]
 }
 nodes [
  Constant3Vector() {constant=<rgba 1 1 2 0>}
  n1 = FunctionInput() {bUsePreviewValueAsDefault=true, inputName=Tiling, inputType=FunctionInput_Vector2, sortPriority=1, previewValue=<4 4 0 1>}
  n2 = Multiply(A=FunctionInput(Preview=TextureCoordinate()) {bUsePreviewValueAsDefault=true, inputName=UVs, inputType=FunctionInput_Vector2}, B=@n1)
  n3 = MaterialFunctionCall(Float2=Ceil(None=@n2)) {materialFunction=&BreakOutFloat2Components}
  n4 = Constant() {r=3}
  Fmod(A=@n3.R, B=@n4)
  n5 = FunctionInput(Preview=CameraVectorWS()) {bUsePreviewValueAsDefault=true, inputName=View}
  n6 = Transform(None=Constant3Vector() {constant=<rgba 1 0 0 0>}) {transformSourceType=TRANSFORMSOURCE_Local}
  n7 = Multiply(A=Transform(None=Constant3Vector() {constant=<rgba 0 1 0 0>}) {transformSourceType=TRANSFORMSOURCE_Local}) {constB=-1}
  n8 = Transform(None=VertexNormalWS()) {transformSourceType=TRANSFORMSOURCE_World, transformType=TRANSFORM_Local}
  n9 = ComponentMask(None=@n8) {r=true}
  n10 = Clamp(None=@n9)
  n11 = Constant3Vector() {constant=<rgba 1 0 0 0>}
  n12 = ComponentMask(None=@n8) {g=true}
  n13 = Clamp(None=@n12)
  n14 = MaterialFunctionCall(Input=@n9, Input Low=Constant(), Input High=Constant() {r=-1}) {materialFunction=&RemapValue_0-1}
  n15 = LinearInterpolate(
   A=LinearInterpolate(A=LinearInterpolate(A=@n6, B=@n7, Alpha=@n10), B=Transform(None=@n11) {transformSourceType=TRANSFORMSOURCE_Local}, Alpha=@n13)
   B=Transform(None=Constant3Vector() {constant=<rgba 0 1 0 0>}) {transformSourceType=TRANSFORMSOURCE_Local}
   Alpha=@n14.Result
  ) {constB=0}
  n16 = Multiply(A=Transform(None=Constant3Vector() {constant=<rgba 1 0 0 0>}) {transformSourceType=TRANSFORMSOURCE_Local}) {constB=-1}
  n17 = MaterialFunctionCall(Input=@n12, Input Low=Constant(), Input High=Constant() {r=-1}) {materialFunction=&RemapValue_0-1}
  n18 = Transform(None=Constant3Vector() {constant=<rgba 1 0 0 0>}) {transformSourceType=TRANSFORMSOURCE_Local}
  n19 = MaterialFunctionCall(Input=ComponentMask(None=@n8) {b=true}, Input Low=Constant(), Input High=Constant() {r=-1}) {materialFunction=&RemapValue_0-1}
  n20 = Transform(None=Constant3Vector() {constant=<rgba 0 1 0 0>}) {transformSourceType=TRANSFORMSOURCE_Local}
  n21 = Multiply(A=Transform(None=Constant3Vector() {constant=<rgba 0 0 1 0>}) {transformSourceType=TRANSFORMSOURCE_Local}) {constB=-1}
  n22 = Transform(None=Constant3Vector() {constant=<rgba 0 0 1 0>}) {transformSourceType=TRANSFORMSOURCE_Local}
  n23 = Constant3Vector() {constant=<rgba 0 0 1 0>}
  n24 = LinearInterpolate(
   A=LinearInterpolate(A=LinearInterpolate(A=@n20, B=@n21, Alpha=@n10), B=Multiply(A=@n22) {constB=-1}, Alpha=@n13)
   B=Multiply(A=Transform(None=@n23) {transformSourceType=TRANSFORMSOURCE_Local}) {constB=-1}
   Alpha=@n14.Result
  ) {constB=0}
  n25 = Multiply(A=Transform(None=Constant3Vector() {constant=<rgba 0 0 1 0>}) {transformSourceType=TRANSFORMSOURCE_Local}) {constB=-1}
  n26 = Constant3Vector() {constant=<rgba 0 1 0 0>}
  n27 = Transform(None=Constant3Vector() {constant=<rgba 0 0 1 0>}) {transformSourceType=TRANSFORMSOURCE_Local}
  n28 = Transform(None=Constant3Vector() {constant=<rgba 1 0 0 0>}) {transformSourceType=TRANSFORMSOURCE_Local}
  n29 = Constant3Vector() {constant=<rgba 0 1 0 0>}
  n30 = Constant3Vector() {constant=<rgba 1 0 0 0>}
  n31 = LinearInterpolate(
   A=LinearInterpolate(A=LinearInterpolate(A=@n27, B=@n28, Alpha=@n10), B=Transform(None=@n29) {transformSourceType=TRANSFORMSOURCE_Local}, Alpha=@n13)
   B=Multiply(A=Transform(None=@n30) {transformSourceType=TRANSFORMSOURCE_Local}) {constB=-1}
   Alpha=@n14.Result
  ) {constB=0}
  n32 = Multiply(A=Transform(None=Constant3Vector() {constant=<rgba 0 1 0 0>}) {transformSourceType=TRANSFORMSOURCE_Local}) {constB=-1}
  n33 = Transform(None=Constant3Vector() {constant=<rgba 0 0 1 0>}) {transformSourceType=TRANSFORMSOURCE_Local}
  n34 = MaterialFunctionCall!(
   VectorToTransform=@n5
   BasisX=LinearInterpolate(A=LinearInterpolate(A=@n15, B=@n16, Alpha=@n17.Result), B=Multiply(A=@n18) {constB=-1}, Alpha=@n19.Result)
   BasisY=LinearInterpolate(
    A=LinearInterpolate(A=@n24, B=@n25, Alpha=@n17.Result)
    B=Transform(None=@n26) {transformSourceType=TRANSFORMSOURCE_Local}
    Alpha=@n19.Result
   )
   BasisZ=LinearInterpolate(A=LinearInterpolate(A=@n31, B=@n32, Alpha=@n17.Result), B=Multiply(A=@n33) {constB=-1}, Alpha=@n19.Result)
  ) {materialFunction=/Engine/Functions/Engine_MaterialFunctions02/Math/InverseTransformMatrix}
  n35 = Transform(None=@n5) {transformSourceType=TRANSFORMSOURCE_World, transformType=TRANSFORM_Tangent}
  n36 = Constant() {r=1}
  n37 = Constant() {r=1}
  n38 = Divide(B=FunctionInput(Preview=Constant() {r=1}) {bUsePreviewValueAsDefault=true, inputName=RoomZscale, inputType=FunctionInput_Scalar})
  n39 = Multiply(
   A=Constant3Vector() {constant=<rgba -1 1 1 0>}
   B=Multiply(
    A=StaticSwitchParameter(True=@n34.Transformed Vector 3, False=@n35) {parameterName=UseBoxTangentTransforms}
    B=MaterialFunctionCall(X=@n36, Y=@n37, Z=@n38) {materialFunction=&MakeFloat3}.Result
   )
  )
  n40 = Divide(B=@n39) {desc=id}
  n41 = Frac(None=@n2) {desc=f}
  n42 = Constant2Vector() {g=-2, r=2}
  n43 = AppendVector(A=Subtract(A=Multiply(A=@n41, B=@n42), B=Constant2Vector() {g=-1, r=1}), B=Constant() {r=-1})
  n44 = MaterialFunctionCall(Float3=Subtract(A=Abs(None=@n40), B=Multiply(A=@n40, B=@n43)) {desc=k}) {materialFunction=&BreakOutFloat3Components}
  n45 = Min(A=Min(A=@n44.R, B=@n44.R), B=@n44.B)
  n46 = MaterialFunctionCall(Float3=Add(A=Multiply(A=@n45, B=@n39), B=@n43) {desc="pos.xyz += kMin * In.dir;"}) {materialFunction=&BreakOutFloat3Components}
  n47 = MaterialFunctionCall(X=@n46.B, Y=@n46.B, Z=@n46.B) {materialFunction=&MakeFloat3}
  n48 = Constant() {r=2}
  n49 = Fmod(A=@n3.G, B=@n48)
  n50 = Multiply(
   A=LinearInterpolate(A=Constant3Vector() {constant=<rgba -1 1 1 0>}, B=Constant3Vector() {constant=<rgba 1 -1 1 0>}, Alpha=Fmod(A=@n3.R, B=@n48))
   B=LinearInterpolate(A=Constant3Vector() {constant=<rgba 1 1 1 0>}, B=Constant3Vector() {constant=<rgba -1 1 1 0>}, Alpha=@n49)
  )
  n51 = Multiply(A=@n47.Result, B=@n50)
  n52 = MaterialFunctionCall(XYZ=@n51) {materialFunction=&Swizzle}
  n53 = FunctionInput(Preview=StaticBool()) {bUsePreviewValueAsDefault=true, inputName=Randomize Rotation, inputType=FunctionInput_StaticBool, sortPriority=2, previewValue=<4 4 0 1>}
  FunctionOutput(None=StaticSwitch(True=LinearInterpolate(A=@n51, B=@n52.YXZ, Alpha=@n49), False=@n47.Result, Value=@n53)) {outputName=UVW}
  LinearInterpolate(
   A=@n50
   B=MaterialFunctionCall(XYZ=@n50) {materialFunction=&Swizzle}.YXZ
   Alpha=Fmod(A=@n3.G, B=@n4)
  )
 ]
}
