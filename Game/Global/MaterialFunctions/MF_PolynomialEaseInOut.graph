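# Material function, exposed to the library, that remaps a scalar through a polynomial ease-in-out S-curve.
# The Gain output offsets Input by InputBias, saturates it, then evaluates x^n / (x^n + (1-x)^n) with CurveGain as exponent.
# The Gain + Bias output repeats that curve after a Schlick-style rational bias driven by CurveBias, which skews the midpoint.
# Named reroutes organize the graph, and every input falls back to a preview default.
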MaterialFunction /Game/Global/MaterialFunctions/MF_PolynomialEaseInOut
graph {
 properties {bExposeToLibrary=true}
 nodes [
  n1 = FunctionInput(Preview=ComponentMask(None=TextureCoordinate()) {r=true}) {bUsePreviewValueAsDefault=true, inputName=Input, inputType=FunctionInput_Scalar}
  n2 = FunctionInput(Preview=Constant()) {bUsePreviewValueAsDefault=true, inputName=InputBias, inputType=FunctionInput_Scalar, sortPriority=1}
  n3 = Saturate(None=Add(
    A=NamedRerouteDeclaration(None=@n1) {name=Input, nodeColor=<rgba 1 0 0.5156 1>}
    B=NamedRerouteDeclaration(None=@n2) {name=InputBias, nodeColor=<rgba 0 1 0.1094 1>}
   ) {constB=0})
  n4 = NamedRerouteDeclaration(None=FunctionInput(Preview=Constant() {r=1}) {bUsePreviewValueAsDefault=true, inputName=CurveGain, inputType=FunctionInput_Scalar, sortPriority=2}) {name=CurveExponent, nodeColor=<rgba 1 0.75 0 1>}
  n5 = Power(Base=@n3, Exp=@n4) {desc="x^n"}
  n6 = Power(Base=OneMinus(None=@n3), Exp=@n4)
  FunctionOutput(None=Divide(A=@n5, B=Add(A=@n5, B=@n6))) {outputName=Gain}
  n7 = NamedRerouteUsage()
  n8 = Divide(B=FunctionInput(Preview=Constant() {desc=CurveBias, r=0.5}) {bUsePreviewValueAsDefault=true, inputName=CurveBias, inputType=FunctionInput_Scalar, sortPriority=3})
  n9 = Add(A=Multiply(A=Subtract(A=@n8) {constB=2}, B=Subtract(B=@n7) {constB=2}))
  n10 = Saturate(None=Add(A=Divide(A=@n7, B=@n9), B=NamedRerouteUsage()))
  n11 = NamedRerouteUsage()
  n12 = Power(Base=@n10, Exp=@n11) {desc="x^n"}
  n13 = Power(Base=OneMinus(None=@n10), Exp=@n11)
  FunctionOutput(None=Divide(A=@n12, B=Add(A=@n12, B=@n13))) {outputName=Gain + Bias, sortPriority=1}
 ]
}
