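# Seasonal material function that pulses tinted emissive and base color, red by default, while the global LockdownActive collection parameter is raised.
# A Time-driven wave, either a quick two-sine product or a sine eased on one half-cycle, is optionally remapped and also output alone.
# LockdownActive blends in each tint, with optional emissive masking for base color and optional black emissive until lockdown.
# Static switches gate each part, passing attributes through when LockdownPulse is off.
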MaterialFunction /Game/Global/MaterialFunctions/MF_LockdownPulse
graph {nodes [
  n1 = FunctionInput(Preview=SetMaterialAttributes!(MaterialAttributes)) {bUsePreviewValueAsDefault=true, inputName=MA In, inputType=FunctionInput_MaterialAttributes}
  n2 = GetMaterialAttributes(None=@n1) {attributeGetTypes [#1, #2]}
  n3 = NamedRerouteUsage()
  n4 = NamedRerouteUsage()
  FunctionOutput(None=StaticSwitchParameter(
    True=SetMaterialAttributes!(MaterialAttributes=@n2.MaterialAttributes, EmissiveColor=@n3, BaseColor=@n4) {attributeSetTypes [#1, #2]}
    False=@n1
   ) {group=Seasonal, parameterName=LockdownPulse})
  n5 = Time()
  n6 = ConstantBiasScale(None=Sine(None=@n5))
  n7 = ConstantBiasScale(None=Sine(None=@n5) {period=2})
  n8 = Time()
  n9 = ConstantBiasScale(None=Sine(None=@n8) {period=2})
  n10 = Constant() {r=0.0427}
  n11 = Constant() {r=2.3965}
  n12 = Constant() {r=0.8553}
  n13 = StaticSwitchParameter(
   True=Saturate(None=Multiply(A=@n6, B=@n7))
   False=If(
    A=Cosine(None=@n8) {period=2}
    "A > B"=MaterialFunctionCall!(Input=@n9, InputBias=@n10, CurveGain=@n11, CurveBias=@n12) {materialFunction=&/MF_PolynomialEaseInOut}.Gain + Bias
    "A < B"=@n9
   )
  ) {group=Seasonal, parameterName=LockdownUseQuickPulse}
  n14 = LinearInterpolate(
   A=ScalarParameter() {defaultValue=0.1, group=Seasonal, parameterName=LockdownSineLow}
   B=ScalarParameter() {defaultValue=1, group=Seasonal, parameterName=LockdownSineHigh}
   Alpha=@n13
  )
  FunctionOutput(None=NamedRerouteDeclaration(None=StaticSwitchParameter(True=@n14, False=@n13) {defaultValue=true, group=Seasonal, parameterName=LockdownRemapSineRange}) {name=SineWave, nodeColor=<rgba 0.0078 1 0 1>}) {outputName=OscillatingPulseOnly, sortPriority=1}
  NamedRerouteDeclaration(None=@n2.EmissiveColor) {name=EmissiveIn, nodeColor=<rgba 0.3125 0 1 1>}
  NamedRerouteDeclaration(None=@n2.BaseColor) {name=BaseColorIn, nodeColor=<rgba 0.0078 0 1 1>}
  n15 = StaticSwitchParameter(True=Constant(), False=NamedRerouteUsage()) {group=Seasonal, parameterName=LockdownNoEmissiveUntilLockdown}
  n16 = Multiply(
   A=Multiply(A=NamedRerouteUsage(), B=NamedRerouteUsage())
   B=VectorParameter() {group=Seasonal, parameterName=LockdownEmissiveTint, defaultValue=<rgba 1 0.0282 0.02 0>}.RGB
  )
  n17 = CollectionParameter() {collection=&FortniteMaterialParameters, parameterName=LockdownActive}
  NamedRerouteDeclaration(None=StaticSwitchParameter(True=LinearInterpolate(A=@n15, B=@n16, Alpha=@n17), False=NamedRerouteUsage()) {group=Seasonal, parameterName=LockdownPulse_TintEmissive}) {name=EmissiveOut, nodeColor=<rgba 1 0.8203 0 1>}
  n18 = NamedRerouteUsage()
  n19 = Multiply(
   A=Multiply(A=NamedRerouteUsage(), B=NamedRerouteUsage())
   B=VectorParameter() {group=Seasonal, parameterName=LockdownBaseColorTint, defaultValue=<rgba 1 0.1037 0.0885 0>}.RGB
  )
  n20 = NamedRerouteUsage()
  n21 = VectorParameter() {group=Seasonal, parameterName=TintBaseColor_EmissiveMaskChannel, defaultValue=<rgba 0.3333 0.3333 0.3333 0>}
  n22 = CollectionParameter() {collection=&FortniteMaterialParameters, parameterName=LockdownActive}
  n23 = StaticSwitchParameter(True=Multiply(A=DotProduct(A=@n20, B=@n21.RGB), B=@n22), False=@n22) {group=Seasonal, parameterName=LockdownPulse_TintBaseColor_MaskByEmissive}
  NamedRerouteDeclaration(None=StaticSwitchParameter(True=LinearInterpolate(A=@n18, B=@n19, Alpha=@n23), False=NamedRerouteUsage()) {group=Seasonal, parameterName=LockdownPulse_TintBaseColor}) {name=BaseColorOut, nodeColor=<rgba 0 1 0.2734 1>}
 ]}
