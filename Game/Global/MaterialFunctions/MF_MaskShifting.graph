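# MF_MaskShifting is a material function that slides a remap window across a scalar mask, with controls for position, width and output scale.
# The window's low edge is Amount minus Start Offset, and its high edge adds Range to that.
# RemapValueRange maps Mask In from this window to between zero and Sharpness, with no clamp added in this function.
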MaterialFunction /Game/Global/MaterialFunctions/MF_MaskShifting
graph {nodes [
  n1 = Subtract(
   A=FunctionInput() {bUsePreviewValueAsDefault=true, inputName=Amount, inputType=FunctionInput_Scalar, sortPriority=1}
   B=FunctionInput() {bUsePreviewValueAsDefault=true, inputName=Start Offset, inputType=FunctionInput_Scalar, sortPriority=3}
  )
  n2 = FunctionInput() {bUsePreviewValueAsDefault=true, inputName=Range, inputType=FunctionInput_Scalar, sortPriority=2}
  FunctionOutput(None=MaterialFunctionCall(
    Input=FunctionInput() {bUsePreviewValueAsDefault=true, inputName=Mask In, inputType=FunctionInput_Scalar}
    Input Low=@n1
    Input High=Add(A=@n1, B=@n2)
    Target Low=Constant()
    Target High=FunctionInput() {bUsePreviewValueAsDefault=true, inputName=Sharpness, inputType=FunctionInput_Scalar, sortPriority=4}
   ) {materialFunction=&RemapValueRange}.Result)
 ]}
