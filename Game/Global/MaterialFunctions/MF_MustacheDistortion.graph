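# Material function that warps UVs with mustache-style lens distortion around a configurable center, outputting coordinates and a mask.
# Radial K and tangential P terms built from centered offsets and radius follow a Brown-Conrady-like model, with a static switch between K1-only and higher-order paths.
# A scale bias rescales coordinates about the middle, feeding sphere-mask edges and mirrored, unmodified or clamped tiling.
# K3 is a zero constant, and its sixth-power sum is left unconnected.
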MaterialFunction /Game/Global/MaterialFunctions/MF_MustacheDistortion
graph {nodes [
  n1 = NamedRerouteUsage()
  n2 = Power(Base=NamedRerouteUsage())
  n3 = NamedRerouteUsage()
  n4 = Add(
   A=NamedRerouteDeclaration(None=Multiply(A=@n1, B=@n2)) {name=K1_Result, nodeColor=<rgba 0 0.1885 0.5885 1>}
   B=Multiply(A=NamedRerouteUsage(), B=Power(Base=@n3) {constExponent=4})
  )
  n5 = NamedRerouteUsage()
  Add(A=@n4, B=Multiply(A=NamedRerouteUsage(), B=Power(Base=@n5) {constExponent=6}))
  FunctionOutput(None=NamedRerouteUsage())
  FunctionOutput(None=NamedRerouteUsage()) {outputName=Mask, sortPriority=100}
  n6 = FunctionInput(Preview=TextureCoordinate()) {bUsePreviewValueAsDefault=true, inputName=UV, inputType=FunctionInput_Vector2, sortPriority=30}
  NamedRerouteDeclaration(None=ComponentMask(None=@n6) {r=true}) {name=Xd, nodeColor=<rgba 0.7231 0.007 0.007 1>}
  n7 = FunctionInput() {bUsePreviewValueAsDefault=true, inputName=Center, inputType=FunctionInput_Vector2, sortPriority=40, previewValue=<0.5 0.5 0 1>}
  NamedRerouteDeclaration(None=ComponentMask(None=@n7) {r=true}) {name=Xc, nodeColor=<rgba 0.7231 0.007 0.007 1>}
  NamedRerouteDeclaration(None=Subtract(A=NamedRerouteUsage(), B=NamedRerouteUsage())) {name=Xp, nodeColor=<rgba 0.7231 0.007 0.007 1>}
  n8 = ComponentMask(None=NamedRerouteUsage()) {r=true}
  n9 = NamedRerouteUsage()
  n10 = Add(A=Power(Base=NamedRerouteUsage()), B=Multiply(A=Power(Base=@n9)) {constB=2}) {constB=2}
  NamedRerouteDeclaration(None=Power(Base=Multiply(A=@n8, B=@n10))) {name=2_X, nodeColor=<rgba 0.7109 0 1 1>}
  NamedRerouteDeclaration(None=Subtract(A=NamedRerouteUsage(), B=NamedRerouteUsage())) {name=Yp, nodeColor=<rgba 0 0.9911 0.0003 1>}
  NamedRerouteDeclaration(None=FunctionInput() {bUsePreviewValueAsDefault=true, inputName=P2, inputType=FunctionInput_Vector2, sortPriority=64}) {name=P2, nodeColor=<rgba 0 0.8594 1 1>}
  NamedRerouteDeclaration(None=FunctionInput() {bUsePreviewValueAsDefault=true, inputName=P3, inputType=FunctionInput_Vector2, sortPriority=66}) {name=P3, nodeColor=<rgba 0 0.8594 1 1>}
  n11 = NamedRerouteUsage()
  n12 = Multiply(A=Multiply(A=ComponentMask(None=@n11) {r=true}) {constB=2}, B=NamedRerouteUsage())
  n13 = NamedRerouteUsage()
  n14 = NamedRerouteUsage()
  n15 = NamedRerouteUsage()
  n16 = NamedRerouteUsage()
  n17 = NamedRerouteUsage()
  n18 = Add(
   A=Multiply(A=ComponentMask(None=@n14) {r=true}, B=Power(Base=@n15))
   B=Multiply(A=ComponentMask(None=@n16) {r=true}, B=Power(Base=@n17) {constExponent=4})
  ) {constA=1}
  NamedRerouteDeclaration(None=Multiply(A=Multiply(A=@n12, B=@n13), B=Add(B=@n18) {constA=1})) {name=3_X, nodeColor=<rgba 0.7812 1 0 1>}
  n19 = FunctionInput() {bUsePreviewValueAsDefault=true, inputName=P4, inputType=FunctionInput_Vector2, sortPriority=68, previewValue=<0 4 0 1>}
  NamedRerouteDeclaration(None=@n19) {name=P4, nodeColor=<rgba 0 0.8594 1 1>}
  n20 = NamedRerouteUsage()
  NamedRerouteDeclaration(None=Multiply(A=NamedRerouteUsage(), B=ComponentMask(None=@n20) {g=true})) {name=1_Y, nodeColor=<rgba 1 0 0.4738 1>}
  n21 = Multiply(A=ComponentMask(None=NamedRerouteUsage()) {g=true}) {constB=2}
  n22 = NamedRerouteUsage()
  NamedRerouteDeclaration(None=Multiply(A=Multiply(A=@n21, B=@n22), B=NamedRerouteUsage())) {name=2_Y, nodeColor=<rgba 1 0 0.4738 1>}
  NamedRerouteDeclaration(None=ComponentMask(None=@n6) {g=true}) {name=Yd, nodeColor=<rgba 0 0.9911 0.0003 1>}
  n23 = ComponentMask(None=NamedRerouteUsage()) {g=true}
  n24 = NamedRerouteUsage()
  n25 = Add(A=Power(Base=NamedRerouteUsage()), B=Multiply(A=Power(Base=@n24)) {constB=2})
  n26 = NamedRerouteUsage()
  n27 = NamedRerouteUsage()
  n28 = NamedRerouteUsage()
  n29 = NamedRerouteUsage()
  n30 = Add(
   A=Multiply(A=ComponentMask(None=@n26) {g=true}, B=Power(Base=@n27))
   B=Multiply(A=ComponentMask(None=@n28) {g=true}, B=Power(Base=@n29) {constExponent=4})
  )
  NamedRerouteDeclaration(None=Multiply(A=Multiply(A=@n23, B=@n25), B=Add(B=@n30) {constA=1})) {name=3_Y, nodeColor=<rgba 1 0 0.4738 1>}
  n31 = NamedRerouteUsage()
  NamedRerouteDeclaration(None=Multiply(A=NamedRerouteUsage(), B=ComponentMask(None=@n31) {r=true})) {name=1_X, nodeColor=<rgba 0.2422 0 1 1>}
  n32 = Constant() {r=0.5}
  n33 = NamedRerouteUsage()
  n34 = Add(A=FunctionInput() {bUsePreviewValueAsDefault=true, inputName=UV Scale Bias, inputType=FunctionInput_Scalar, sortPriority=300})
  n35 = Constant() {r=0.5}
  n36 = Subtract(A=Add(A=Divide(A=@n33, B=@n34), B=@n35), B=Divide(A=@n35, B=@n34) {desc=Scale the center offset by the scaling factor}) {desc=Re offset to maintain center}
  n37 = MaterialFunctionCall(Float2=@n36) {materialFunction=&BreakOutFloat2Components}
  NamedRerouteDeclaration(None=Multiply(A=SphereMask(A=@n32, B=@n37.R) {attenuationRadius=0.5}, B=SphereMask(A=@n37.G, B=@n32) {attenuationRadius=0.5})) {name=Mask, nodeColor=<rgba 1 0.8203 0 1>}
  n38 = FunctionInput(Preview=StaticBool()) {
   bUsePreviewValueAsDefault=true
   description="by default, this function only uses K1, enable to use K2 and K3"
   inputName=Enable K2
   inputType=FunctionInput_StaticBool
   sortPriority=500
  }
  n39 = NamedRerouteUsage()
  n40 = NamedRerouteUsage()
  n41 = Add(A=NamedRerouteUsage(), B=Add(A=Multiply(A=@n39, B=@n40), B=NamedRerouteUsage()))
  n42 = NamedRerouteUsage()
  n43 = NamedRerouteUsage()
  n44 = Add(A=Add(A=NamedRerouteUsage(), B=Multiply(A=@n42, B=@n43)), B=NamedRerouteUsage())
  n45 = Add(A=Add(A=NamedRerouteUsage(), B=NamedRerouteUsage()), B=Add(A=NamedRerouteUsage(), B=NamedRerouteUsage()))
  n46 = Add(A=Add(A=NamedRerouteUsage(), B=NamedRerouteUsage()), B=Add(A=NamedRerouteUsage(), B=NamedRerouteUsage()))
  NamedRerouteDeclaration(None=StaticSwitch(
    True=MaterialFunctionCall(X=@n45, Y=@n46) {materialFunction=&MakeFloat2}.Result
    False=MaterialFunctionCall(X=@n41, Y=@n44) {materialFunction=&MakeFloat2}.Result
    Value=@n38
   )) {name=Distorted Coords, nodeColor=<rgba 0.125 0 1 1>}
  NamedRerouteDeclaration(None=@n36) {name=Scaled Coords, nodeColor=<rgba 0 0.8125 1 1>}
  n47 = NamedRerouteUsage()
  n48 = Multiply(A=NamedRerouteUsage()) {constB=0.5}
  n49 = Floor(None=Add(A=@n48) {constB=0.5})
  n50 = Multiply(A=Abs(None=Subtract(A=@n48, B=@n49))) {constB=2}
  n51 = NamedRerouteUsage()
  n52 = FunctionInput(Preview=StaticBool() {value=true}) {bUsePreviewValueAsDefault=true, description=True by default, inputName=Mirror Tiling, inputType=FunctionInput_StaticBool, sortPriority=1000}
  n53 = FunctionInput(Preview=StaticBool()) {
   bUsePreviewValueAsDefault=true
   description="Overrides Mirror Tiling, false by default"
   inputName=Clamp Tiling
   inputType=FunctionInput_StaticBool
   sortPriority=2000
  }
  NamedRerouteDeclaration(None=StaticSwitch(True=Clamp(None=@n47) {maxDefault=0.999, minDefault=0.001}, False=StaticSwitch(True=@n50, False=@n51, Value=@n52), Value=@n53)) {name=Final Coords, nodeColor=<rgba 1 0 0.9141 1>}
  NamedRerouteDeclaration(None=ComponentMask(None=@n7) {g=true}) {name=Yc, nodeColor=<rgba 0 0.9911 0.0003 1>}
  n54 = FunctionInput() {bUsePreviewValueAsDefault=true, inputName=K1, inputType=FunctionInput_Vector2, sortPriority=50, previewValue=<1 1 0 1>}
  NamedRerouteDeclaration(None=@n54) {name=K1, nodeColor=<rgba 0.7812 1 0 1>}
  NamedRerouteDeclaration(None=FunctionInput() {bUsePreviewValueAsDefault=true, inputName=P1, inputType=FunctionInput_Vector2, sortPriority=62}) {name=P1, nodeColor=<rgba 0 0.8594 1 1>}
  n55 = FunctionInput() {bUsePreviewValueAsDefault=true, inputName=K2, inputType=FunctionInput_Vector2, sortPriority=52, previewValue=<-2 -2 0 1>}
  NamedRerouteDeclaration(None=@n55) {name=K2, nodeColor=<rgba 0 0.3203 1 1>}
  NamedRerouteDeclaration(None=Constant()) {name=K3, nodeColor=<rgba 0 0.3203 1 1>}
  n56 = Power(Base=Abs(None=NamedRerouteUsage()))
  n57 = Power(Base=Abs(None=NamedRerouteUsage()))
  NamedRerouteDeclaration(None=SquareRoot(None=Add(A=@n56, B=@n57))) {name=R, nodeColor=<rgba 0.6973 0.235 1 1>}
  NamedRerouteDeclaration(None=@n4) {name=Ks, nodeColor=<rgba 0 0.1885 0.5885 1>}
 ]}
