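# Material function that fakes refractive translucency on opaque materials, with facet-like interior detail.
# It refracts the view ray into a bounds-scaled interior cubemap room to fetch interior normals, optionally blended with tiled detail normals.
# These drive skylight refraction and reflection, total internal reflection, sun glints and Schlick-phase scattering, combined with tint and gradients into emissive.
# Mobile gets cheaper specular and extra lighting; a static switch optionally alpha-blends with the input material.
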MaterialFunction /Game/Global/MaterialFunctions/MF_OpaqueFakeTranslucency
graph {nodes [
  n1 = NamedRerouteUsage()
  n2 = Multiply(
   A=MaterialFunctionCall(Input=ComponentMask(None=@n1) {r=true}, Input Low=Constant() {r=0.99}, Input High=Constant() {r=1}) {materialFunction=&RemapValue_0-1}.Result
   B=Constant3Vector() {constant=<rgba 0 0 1 0>}
  )
  n3 = NamedRerouteUsage()
  n4 = Multiply(
   A=MaterialFunctionCall(Input=ComponentMask(None=@n3) {g=true}, Input Low=Constant() {r=0.99}, Input High=Constant() {r=1}) {materialFunction=&RemapValue_0-1}.Result
   B=Constant3Vector() {constant=<rgba -1 0 0 0>}
  )
  n5 = MaterialFunctionCall(Input=ComponentMask(None=NamedRerouteUsage()) {b=true}, Input Low=Constant() {r=0.99}, Input High=Constant() {r=1}) {materialFunction=&RemapValue_0-1}
  n6 = Constant3Vector() {constant=<rgba 0 1 0 0>}
  n7 = MaterialFunctionCall(Input=ComponentMask(None=NamedRerouteUsage()) {g=true}, Input Low=Constant() {r=-0.99}, Input High=Constant() {r=-1}) {materialFunction=&RemapValue_0-1}
  n8 = Constant3Vector() {constant=<rgba 1 0 0 0>}
  n9 = MaterialFunctionCall(Input=ComponentMask(None=NamedRerouteUsage()) {b=true}, Input Low=Constant() {r=-0.99}, Input High=Constant() {r=-1}) {materialFunction=&RemapValue_0-1}
  n10 = Constant3Vector() {constant=<rgba 0 -1 0 0>}
  Add(A=Add(A=Add(A=@n2, B=@n4), B=Multiply(A=@n5.Result, B=@n6)), B=Add(A=Multiply(A=@n7.Result, B=@n8), B=Multiply(A=@n9.Result, B=@n10)))
  n11 = NamedRerouteUsage()
  ComponentMask(None=@n11) {g=true, r=true}
  Constant()
  n12 = NamedRerouteUsage()
  n13 = Multiply(A=NamedRerouteUsage()) {constB=0.5}
  n14 = DotProduct(A=NamedRerouteUsage(), B=SkyAtmosphereLightDirection())
  n15 = Multiply(A=Multiply(A=Multiply(A=@n13) {constB=0.5}) {constB=0.5}, B=LinearInterpolate(Alpha=Abs(None=@n14)) {constA=0.5})
  n16 = ShadingPathSwitch(Default=Constant(), Mobile=@n15)
  n17 = Add(A=Add(A=NamedRerouteUsage(), B=NamedRerouteUsage()), B=NamedRerouteUsage())
  n18 = SetMaterialAttributes!(
   MaterialAttributes=@n12
   Normal=NamedRerouteUsage()
   Metallic=Constant()
   Roughness=Constant() {r=0.15}
   EmissiveColor=NamedRerouteDeclaration(None=Add(A=@n16, B=@n17) {constB=0}) {name=Emissive, nodeColor=<rgba 0 1 0.0391 1>}
   Opacity=Constant() {r=0.7}
   SubsurfaceColor=@n15
  ) {attributeSetTypes [
    #1
    #2
    #3
    #4
    #5
    #6
   ]}
  n19 = FunctionInput() {bUsePreviewValueAsDefault=true, inputName=Alpha, inputType=FunctionInput_Scalar, sortPriority=2, previewValue=<1 0 0 1>}
  FunctionOutput(None=StaticSwitchParameter(True=BlendMaterialAttributes(A=@n12, B=@n18, Alpha=@n19), False=@n18) {group=FakeTranslucency, parameterName=FT_BlendMaterial, sortPriority=100})
  LinearInterpolate(A=NamedRerouteUsage(), Alpha=NamedRerouteUsage()) {constA=1, constB=0.5}
  n20 = Multiply(A=CameraVectorWS()) {constB=-1}
  n21 = NamedRerouteUsage()
  n22 = NamedRerouteDeclaration(None=MaterialFunctionCall!(
    Ray Direction=Normalize(VectorInput=@n20)
    Surface Normal=Transform(None=@n21)
    Refractive Index Origin=Constant() {r=1}
    Refractive Index Target=Constant() {r=1.5}
   ) {materialFunction=/Engine/Functions/Engine_MaterialFunctions01/Vectors/Refract}.Vector) {name=RefractVectorSurface, nodeColor=<rgba 0.2188 0 1 1>}
  n23 = Custom!(None) {code="GetPrimitiveData(Parameters).InstanceLocalBoundsExtent.xyz", description=BoundsExtent}
  n24 = MaterialFunctionCall() {materialFunction=&ObjectScale}
  n25 = Divide(
   A=Transform(None=Multiply(A=@n22) {constB=-1}) {transformSourceType=TRANSFORMSOURCE_World, transformType=TRANSFORM_Instance}
   B=Normalize(VectorInput=Multiply(A=@n23, B=@n24.Scale XYZ))
  )
  NamedRerouteDeclaration(None=MaterialFunctionCall!(
    UVs=TextureCoordinate()
    View=Transform(None=@n25) {transformSourceType=TRANSFORMSOURCE_Instance}
    RoomZscale=ScalarParameter() {defaultValue=1, group=FakeTranslucency, parameterName=FT_RoomDepth, sortPriority=3}
    Tiling=ScalarParameter() {defaultValue=1, group=FakeTranslucency, parameterName=FT_RoomTile, sortPriority=2}
    Randomize Rotation
   ) {materialFunction=&/MF_InteriorCubemapUV}.UVW) {name=uvw, nodeColor=<rgba 0.1016 0 1 1>}
  n26 = NamedRerouteUsage()
  n27 = MaterialFunctionCall!(Normal_1=@n26, Normal_2=NamedRerouteUsage()) {materialFunction=&/MF_NormalBlend_Whiteout}
  NamedRerouteDeclaration(None=Transform(None=StaticSwitchParameter(True=@n27.Result, False=@n26) {group=FakeTranslucency, parameterName=FT_UseDetailNormal})) {name=InteriorNormal, nodeColor=<rgba 1 0.3984 0 1>}
  MaterialExpressionNormalize_2 = Normalize(VectorInput=SkyAtmosphereLightDirection())
  MaterialExpressionCameraVectorWS_1 = CameraVectorWS()
  MaterialExpressionNamedRerouteUsage_1 = NamedRerouteUsage()
  MaterialExpressionConstant_28 = Constant() {r=0.3}
  MaterialExpressionConstant3Vector_3 = Constant3Vector() {constant=<rgba 1 1 1 0>}
  n28 = Custom!(
   L=@MaterialExpressionNormalize_2
   V=@MaterialExpressionCameraVectorWS_1
   N=@MaterialExpressionNamedRerouteUsage_1
   roughness=@MaterialExpressionConstant_28
   F0=@MaterialExpressionConstant3Vector_3
  ) {
   code="    float3 H = normalize(L + V);\r\n    float NdotH = saturate(dot(N, H));\r\n    float NdotL = saturate(dot(N, L));\r\n    float NdotV = saturate(dot(N, V));\r\n    float3 F = F0 + (1.0 - F0) * pow(1.0 - NdotV, 5.0);\r\n    float epsilon = .0001;\r\n\r\n\tfloat alpha = roughness * roughness;\r\n    float alphaSqr = alpha * alpha;\r\n\r\n    float denominator = NdotH * NdotH * (alphaSqr - 1.0) + 1.0;\r\n    float D = alphaSqr / (PI * denominator * denominator + epsilon);\r\n\t\r\n\t\r\n\t\r\n    float k = alpha / 2;    \r\n    float g_v = NdotV / ( NdotV*(1 - k ) + k);\r\n    float g_l = NdotL / (NdotL*(1 - k ) + k);\r\n    float G = g_v * g_l;\r\n    return NdotL*(D * G * F) / (4.0 * NdotL * NdotV + epsilon );"
   description=GGX
   inputs [
    {inputName=L, input {expression=&self:MaterialExpressionNormalize_2}}
    {inputName=V, input {expression=&self:MaterialExpressionCameraVectorWS_1}}
    {inputName=N, input {expression=&self:MaterialExpressionNamedRerouteUsage_1}}
    {inputName=roughness, input {expression=&self:MaterialExpressionConstant_28}}
    {
     inputName=F0
     input {expression=&self:MaterialExpressionConstant3Vector_3, mask=1, maskB=1, maskG=1, maskR=1}
    }
   ]
  }
  n29 = Multiply(A=@n28) {constA=1}
  n30 = DotProduct(A=ReflectionVectorWS(CustomWorldNormal=@MaterialExpressionNamedRerouteUsage_1), B=@MaterialExpressionNormalize_2)
  n31 = Multiply(A=Power(Base=Saturate(None=@n30)) {constExponent=24}) {constB=5}
  n32 = NamedRerouteUsage()
  n33 = Multiply(A=Clamp(None=ShadingPathSwitch(Default=@n29, Mobile=@n31)) {maxDefault=40}, B=Multiply(A=@n32, B=@n32))
  n34 = NamedRerouteUsage()
  n35 = NamedRerouteUsage()
  NamedRerouteDeclaration(None=Multiply(A=Multiply(A=@n33, B=@n34), B=LinearInterpolate(Alpha=@n35) {constA=0.5, constB=0.05})) {name=SunInternalReflections, nodeColor=<rgba 0 0.8828 1 1>}
  n36 = GetMaterialAttributes(None=FunctionInput() {inputName=Material, inputType=FunctionInput_MaterialAttributes}) {attributeGetTypes [#1, #7]}
  NamedRerouteDeclaration(None=@n36.Normal) {name=Normal, nodeColor=<rgba 0.4531 0 1 1>}
  n37 = NamedRerouteUsage()
  n38 = NamedRerouteUsage()
  n39 = ScalarParameter() {defaultValue=0.2, group=FakeTranslucency, parameterName=FT_ReflectRoughness}
  n40 = SkyLightEnvMapSample(Direction=NamedRerouteUsage(), Roughness=ScalarParameter() {defaultValue=0.2, group=FakeTranslucency, parameterName=FT_RefractRoughness})
  n41 = LinearInterpolate(
   A=Multiply(A=SkyLightEnvMapSample(Direction=@n38, Roughness=@n39)) {constB=0.75}
   B=Multiply(A=Multiply(A=@n40) {constB=1.2}, B=NamedRerouteUsage())
   Alpha=NamedRerouteUsage()
  )
  NamedRerouteDeclaration(None=Multiply(B=Multiply(A=@n37, B=@n41)) {constA=1}) {name=Sky Refractions, nodeColor=<rgba 0.6875 1 0 1>}
  n42 = Constant() {r=1}
  n43 = ComponentMask(None=NamedRerouteUsage()) {b=true}
  n44 = Saturate(None=OneMinus(None=Multiply(A=@n43) {constB=-1}))
  NamedRerouteDeclaration(None=MaterialFunctionCall!(1=@n42, 2=Constant() {r=0.5}, 3=@n42, 4, A=Multiply(A=@n44, B=@n44)) {materialFunction=/Engine/Functions/Engine_MaterialFunctions02/Lerp_Multiple_Float}.Lerp 3 Inputs) {name=GroundGradient, nodeColor=<rgba 0.4062 0 1 1>}
  n45 = NamedRerouteUsage()
  n46 = Abs(None=NamedRerouteUsage())
  n47 = MaterialFunctionCall(Float3=@n46) {materialFunction=&BreakOutFloat3Components}
  n48 = Max(A=@n47.R, B=@n47.R)
  n49 = MaterialFunctionCall(Float3=Step(Y=Max(A=@n48, B=@n47.B), X=@n46)) {materialFunction=&BreakOutFloat3Components}
  n50 = LinearInterpolate(A=ComponentMask(None=@n45) {b=true, g=true}, B=ComponentMask(None=@n45) {b=true, r=true}, Alpha=@n49.R)
  n51 = ComponentMask(None=@n45) {g=true, r=true}
  n52 = ScalarParameter() {defaultValue=1, group=FakeTranslucency, parameterName=FT_DetailNormalTileX}
  n53 = ScalarParameter() {defaultValue=1, group=FakeTranslucency, parameterName=FT_DetailNormalTileY}
  NamedRerouteDeclaration(None=Multiply(A=LinearInterpolate(A=@n50, B=@n51, Alpha=@n49.G), B=AppendVector(A=@n52, B=@n53))) {name=Detail Normal UVs, nodeColor=<rgba 0.5938 0 1 1>}
  n54 = TextureSampleParameter2D(UVs=NamedRerouteUsage()) {group=FakeTranslucency, parameterName=FT_DetailNormal, samplerType=SAMPLERTYPE_Normal, texture=/Game/Global/Textures/T_facets_N}
  n55 = ComponentMask(None=@n54.RGB) {g=true, r=true}
  n56 = ScalarParameter() {defaultValue=1, group=FakeTranslucency, parameterName=FT_DetailNormalStrength}
  NamedRerouteDeclaration(None=DeriveNormalZ(InXY=Multiply(A=@n55, B=@n56))) {name=Detail Normals, nodeColor=<rgba 0.7109 1 0 1>}
  NamedRerouteDeclaration(None=@n36.MaterialAttributes) {name=InMaterial, nodeColor=<rgba 1 0.6328 0 1>}
  NamedRerouteDeclaration(None=@n36.BaseColor) {name=InColor, nodeColor=<rgba 0 0.2266 1 1>}
  n57 = TextureSampleParameterCube!(UVs=@n11, Apply View MipBias) {
   group=FakeTranslucency
   parameterName=FT_InteriorNormals
   samplerType=SAMPLERTYPE_LinearColor
   sortPriority=0
   texture=/Game/Global/Textures/Cubemaps/TC_InteriorNormals_Tex
  }
  NamedRerouteDeclaration(None=@n57.RGB) {name=Debug, nodeColor=<rgba 0.5234 1 0 1>}
  n58 = Multiply(A=@n57.RGB) {constB=2}
  NamedRerouteDeclaration(None=Normalize(VectorInput=Subtract(A=@n58))) {name=CubeNormal, nodeColor=<rgba 0.6406 1 0 1>}
  NamedRerouteDeclaration(None=@n40) {name=Debug_1, nodeColor=<rgba 0 0.4375 1 1>}
  n59 = Multiply(A=DotProduct(A=NamedRerouteUsage(), B=NamedRerouteUsage())) {constB=-1}
  n60 = Constant() {r=0.73}
  n61 = Constant() {r=0.78}
  NamedRerouteDeclaration(None=LinearInterpolate(Alpha=MaterialFunctionCall(Input=@n59, Input Low=@n60, Input High=@n61) {materialFunction=&RemapValue_0-1}.Result) {constA=0.1, constB=0.9}) {name=TIR, nodeColor=<rgba 0 0.7422 1 1>}
  NamedRerouteDeclaration(None=MaterialFunctionCall(
    Ray Direction=NamedRerouteUsage()
    Surface Normal=NamedRerouteUsage()
    Refractive Index Origin=Constant() {r=1}
    Refractive Index Target=Constant() {r=1.5}
   ) {materialFunction=&Refract}.Vector) {name=RefractVectorInterior, nodeColor=<rgba 1 0 0.9609 1>}
  n62 = NamedRerouteUsage()
  n63 = Constant() {r=-1}
  n64 = Constant() {r=1}
  NamedRerouteDeclaration(None=LinearInterpolate(
    A=ScalarParameter() {defaultValue=0.5, group=FakeTranslucency, parameterName=FT_GradientLow}
    B=ScalarParameter() {defaultValue=1, group=FakeTranslucency, parameterName=FT_GradientHigh}
    Alpha=MaterialFunctionCall(Input=@n62, Input Low=@n63, Input High=@n64) {materialFunction=&RemapValue_0-1}.Result
   ) {constA=0.45}) {name=Depth Gradient, nodeColor=<rgba 1 0 0.2812 1>}
  n65 = VectorParameter() {group=FakeTranslucency, parameterName=FT_ColorTint, sortPriority=-2, defaultValue=<rgba 0.0272 0.5312 0.5101 1>}
  n66 = NamedRerouteUsage()
  n67 = ScalarParameter() {group=FakeTranslucency, parameterName=FT_BaseColorBlend, sortPriority=-1}
  NamedRerouteDeclaration(None=Desaturation(None=LinearInterpolate(A=@n65.RGB, B=@n66, Alpha=@n67), Fraction=Constant() {r=-0.2})) {name=Tint, nodeColor=<rgba 0.3125 1 0 1>}
  NamedRerouteDeclaration(None=ReflectionVectorWS(CustomWorldNormal=NamedRerouteUsage())) {name=ReflectionVector, nodeColor=<rgba 1 0 0.3281 1>}
  n68 = Constant3Vector() {constant=<rgba 0 0 100 0>}
  NamedRerouteDeclaration(None=Multiply(A=SkyAtmosphereLightIlluminance!(World Position=@n68)) {constB=0.5}) {name=DirectionalLightColor, nodeColor=<rgba 0 0.4844 1 1>}
  n69 = DotProduct(A=SkyAtmosphereLightDirection(), B=NamedRerouteUsage())
  n70 = ScalarParameter() {defaultValue=-0.985, group=FakeTranslucency, parameterName=FT_Phase}
  n71 = Clamp(None=Multiply(
    A=NamedRerouteUsage()
    B=MaterialFunctionCall!(Cosine=@n69, Anisotropy=@n70) {materialFunction=/Engine/Functions/Engine_MaterialFunctions01/Volumetrics/SchlickPhase}.Result
   )) {maxDefault=20}
  n72 = NamedRerouteUsage()
  n73 = NamedRerouteUsage()
  NamedRerouteDeclaration(None=Multiply(A=Multiply(A=@n71, B=@n72) {constB=2}, B=LinearInterpolate(Alpha=@n73) {constA=0.2})) {name=SunNscatter, nodeColor=<rgba 1 0.7031 0 1>}
  NamedRerouteUsage()
  NamedRerouteUsage()
  TextureCoordinate()
  n74 = WorldPosition()
  Transform(None=CrossProduct(A=DDX(Value=@n74.XYZ), B=DDY(Value=@n74.XYZ))) {transformSourceType=TRANSFORMSOURCE_World, transformType=TRANSFORM_Tangent}
 ]}
