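# Material function that triplanar-projects a normal map using pre-skinned local position and normal, so the mapping follows skinned meshes.
# A static switch picks WorldAlignedNormals_HighQuality or cheaper hand-built projections, with contrast-adjusted axis blending.
# Results pass through InverseTransformMatrix, alongside a tangent basis built from screen-space derivatives and UVs.
# XYZ, XY, Z and flat-top outputs negate X and feed G into both Y and Z.
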MaterialFunction /Game/Global/MaterialFunctions/MF_PreskinnedWorldAlignedTextureNormal
graph {nodes [
  n1 = FunctionInput() {bUsePreviewValueAsDefault=true, description=Size of texture in worldspace units, inputName=TextureSize, previewValue=<4 4 4 4>}
  n2 = Multiply(A=Abs(None=@n1)) {constA=-1, constB=-1}
  n3 = FunctionInput(Preview=VertexInterpolator(VS=PreSkinnedPosition()).PS) {bUsePreviewValueAsDefault=true, inputName=PreskinnedLocalPosition, sortPriority=100}
  n4 = FunctionInput(Preview=TextureObject() {samplerType=SAMPLERTYPE_Normal, texture=&T_Sphere_N}) {inputName=TextureObject, inputType=FunctionInput_Texture2D}
  n5 = FunctionInput(Preview=VertexInterpolator(VS=PreSkinnedNormal()).PS) {bUsePreviewValueAsDefault=true, inputName=PreskinnedLocalNormal, previewValue=<0 0 1 1>}
  n6 = Multiply(A=@n5, B=Constant() {r=1})
  n7 = ComponentMask(None=@n6) {b=true}
  n8 = ScalarParameter() {defaultValue=1, parameterName=Blend Contrast}
  n9 = MaterialFunctionCall(In=Abs(None=@n7), Contrast=@n8) {materialFunction=&CheapContrast}
  n10 = ComponentMask(None=@n6) {r=true}
  n11 = MaterialFunctionCall(In=Abs(None=@n10), Contrast=@n8) {materialFunction=&CheapContrast}
  n12 = MaterialFunctionCall!(Negative World Position Divisor=@n2, World Position=@n3, Texture=@n4, Z lerp Alpha=@n9.Result, WorldNormal=@n5, X lerp alpha=@n11.Result) {materialFunction=/Engine/Functions/Engine_MaterialFunctions02/WorldAlignedNormals_HighQuality}
  n13 = Divide(A=@n3, B=@n2)
  n14 = TextureSample(UVs=ComponentMask(None=@n13) {b=true, r=true}, Tex=@n4) {samplerType=SAMPLERTYPE_Normal}
  n15 = Multiply(A=@n14.R) {constB=-1}
  n16 = Multiply(A=@n14.B, B=ComponentMask(None=@n6) {g=true})
  n17 = TextureSample(UVs=ComponentMask(None=@n13) {b=true, g=true}, Tex=@n4) {samplerType=SAMPLERTYPE_Normal}
  n18 = Multiply(A=@n10, B=@n17.B)
  n19 = Multiply(A=@n17.R) {constB=-1}
  n20 = LinearInterpolate(
   A=AppendVector(A=AppendVector(A=@n15, B=@n16), B=Multiply(A=@n14.G) {constB=-1})
   B=AppendVector(A=AppendVector(A=@n18, B=@n19), B=Multiply(A=@n17.G) {constB=-1})
   Alpha=@n11.Result
  )
  n21 = TextureSample(UVs=ComponentMask(None=@n13) {g=true, r=true}, Tex=@n4) {samplerType=SAMPLERTYPE_Normal}
  n22 = Constant3Vector() {constant=<rgba -1 -1 1 0>}
  n23 = AppendVector(A=ComponentMask(None=Multiply(A=@n21.RGB, B=@n22)) {g=true, r=true}, B=Multiply(A=@n21.B, B=@n7))
  n24 = Normalize(VectorInput=LinearInterpolate(A=@n20, B=@n23, Alpha=@n9.Result))
  n25 = FunctionInput(Preview=StaticBool() {desc=Use High quality normals, value=true}) {
   bUsePreviewValueAsDefault=true
   description="Default is true. Adds ~ 20 instructions."
   inputName=Use High Quality Normals
   inputType=FunctionInput_StaticBool
   sortPriority=150
  }
  n26 = MaterialFunctionCall(Float3=MaterialFunctionCall!(
    VectorToTransform=StaticSwitch(True=@n12.World Space Normals XYZ, False=@n24, Value=@n25)
    BasisX=NamedRerouteUsage()
    BasisY=NamedRerouteUsage()
    BasisZ=NamedRerouteUsage()
   ) {materialFunction=/Engine/Functions/Engine_MaterialFunctions02/Math/InverseTransformMatrix}.Transformed Vector 3) {materialFunction=&BreakOutFloat3Components}
  FunctionOutput(None=MaterialFunctionCall(X=Multiply(A=@n26.R) {constB=-1}, Y=@n26.G, Z=@n26.G) {materialFunction=&MakeFloat3}.Result) {outputName=XYZ Texture, sortPriority=1}
  n27 = Normalize(VectorInput=@n20)
  n28 = MaterialFunctionCall(Float3=MaterialFunctionCall(
    VectorToTransform=StaticSwitch(True=@n12.XY Projection World Space Normals, False=@n27, Value=@n25)
    BasisX=NamedRerouteUsage()
    BasisY=NamedRerouteUsage()
    BasisZ=NamedRerouteUsage()
   ) {materialFunction=&InverseTransformMatrix}.Transformed Vector 3) {materialFunction=&BreakOutFloat3Components}
  FunctionOutput(None=MaterialFunctionCall(X=Multiply(A=@n28.R) {constB=-1}, Y=@n28.G, Z=@n28.G) {materialFunction=&MakeFloat3}.Result) {outputName=XY Texture, sortPriority=1}
  n29 = Normalize(VectorInput=@n23)
  n30 = MaterialFunctionCall(Float3=MaterialFunctionCall(
    VectorToTransform=StaticSwitch(True=@n12.Z Projection World Space Normals, False=@n29, Value=@n25)
    BasisX=NamedRerouteUsage()
    BasisY=NamedRerouteUsage()
    BasisZ=NamedRerouteUsage()
   ) {materialFunction=&InverseTransformMatrix}.Transformed Vector 3) {materialFunction=&BreakOutFloat3Components}
  FunctionOutput(None=MaterialFunctionCall(X=Multiply(A=@n30.R) {constB=-1}, Y=@n30.G, Z=@n30.G) {materialFunction=&MakeFloat3}.Result) {outputName=Z Texture, sortPriority=1}
  n31 = Normalize(VectorInput=LinearInterpolate(A=@n20, B=@n5, Alpha=@n9.Result))
  n32 = MaterialFunctionCall(Float3=MaterialFunctionCall(
    VectorToTransform=StaticSwitch(True=@n12.XY Flat Top Projection World Space Normals, False=@n31, Value=@n25)
    BasisX=NamedRerouteUsage()
    BasisY=NamedRerouteUsage()
    BasisZ=NamedRerouteUsage()
   ) {materialFunction=&InverseTransformMatrix}.Transformed Vector 3) {materialFunction=&BreakOutFloat3Components}
  FunctionOutput(None=MaterialFunctionCall(X=Multiply(A=@n32.R) {constB=-1}, Y=@n32.G, Z=@n32.G) {materialFunction=&MakeFloat3}.Result) {outputName=XYZFlatTop, sortPriority=1}
  MaterialFunctionCall!(TextureObject, TextureSize, WorldSpace, WorldNormal, WorldPosition, Use High Quality Normals) {materialFunction=/Volumetrics/Content/PlanetPrototype/Functions/WorldAlignedNormal_HQ}
  n33 = NamedRerouteUsage()
  n34 = NamedRerouteUsage()
  NamedRerouteDeclaration(None=CrossProduct(A=@n33, B=DDX(Value=@n34))) {name=dp1perp, nodeColor=<rgba 0 0.8828 1 1>}
  NamedRerouteDeclaration(None=CrossProduct(A=@n33, B=DDY(Value=@n34))) {name=dp2perp, nodeColor=<rgba 0 0.8828 1 1>}
  NamedRerouteDeclaration(None=@n33) {name=Preskinned_N, nodeColor=<rgba 1 0 0.5391 1>}
  n35 = TextureCoordinate()
  n36 = MaterialFunctionCall(Float2=DDX(Value=@n35)) {materialFunction=&BreakOutFloat2Components}
  n37 = MaterialFunctionCall(Float2=DDY(Value=@n35)) {materialFunction=&BreakOutFloat2Components}
  n38 = Add(A=Multiply(A=NamedRerouteUsage(), B=@n36.R), B=Multiply(A=NamedRerouteUsage(), B=@n37.R)) {desc=T}
  n39 = DotProduct(A=@n38, B=@n38)
  n40 = Add(A=Multiply(A=NamedRerouteUsage(), B=@n36.G), B=Multiply(A=NamedRerouteUsage(), B=@n37.G))
  n41 = DotProduct(A=@n40, B=@n40)
  n42 = Divide(B=SquareRoot(None=Max(A=@n39, B=@n41)))
  NamedRerouteDeclaration(None=Multiply(A=@n42, B=@n40)) {name=Preskinned_U, nodeColor=<rgba 0.4766 1 0 1>}
  NamedRerouteDeclaration(None=Multiply(A=@n38, B=@n42)) {name=Preskinned_V, nodeColor=<rgba 1 0.2109 0 1>}
  NamedRerouteDeclaration(None=@n3) {name=PreskinnedLocation, nodeColor=<rgba 0 0.7187 1 1>}
  NamedRerouteDeclaration(None=@n5) {name=PreskinnedNormalIN, nodeColor=<rgba 1 0.6094 0 1>}
  VertexNormalWS()
 ]}
