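# Library material function offering several S-curve (sigmoid) mappings of a scalar input centered at zero, with outputs differing in slope and instruction cost.
# Outputs are logistic in 0..1 and -1..1 sharing one exponential term, tanh through a Custom HLSL node, and exact and fast arctangent.
# Arctangent variants pre-scale by Pi/2 and post-scale by 2/Pi to fit -1..1.
# A widened LinearGradient drives the input preview to show the curves.
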MaterialFunction /Game/Global/MaterialFunctions/MF_SCurveFunctions
graph {
 properties {
  bExposeToLibrary=true
  description="Various S-Curve/Sigmoid Functions to fit arbitrary signed values into the space between -1 and 1. These vary in cost and slope, check output tooltips for more info!"
 }
 nodes [
  n1 = MaterialFunctionCall() {materialFunction=&LinearGradient}
  MaterialExpressionFunctionInput_0 = FunctionInput(Preview=Multiply(A=Subtract(A=@n1.UGradient) {constB=0.5}) {constB=50}) {description=Value to transform. S-Curve center will be at input 0.0., inputType=FunctionInput_Scalar}
  n2 = Custom!(In=@MaterialExpressionFunctionInput_0) {
   code="return tanh(In);"
   description="Tanh(In)"
   showCode=true
   inputs [{inputName=In, input {expression=&self:MaterialExpressionFunctionInput_0}}]
  }
  FunctionOutput(None=ComponentMask(None=@n2) {r=true}) {
   description="Hyperbolic Tangent function. Approaches -1/1 boundaries roughly around +/- 2.55\r\n\r\nAdds 7 instructions"
   outputName="Tanh Function (-1...1)"
   sortPriority=10
  }
  n3 = Add(A=Exponential!(None=Multiply(A=@MaterialExpressionFunctionInput_0) {constB=-1}))
  FunctionOutput(None=Divide(B=@n3)) {
   description="Logistic function in 0-1 space. 1/(1+e^-x). Soft slope, approaches 0/1 boundary around +/- 4.95\r\n\r\nAdds 5 instructions"
   outputName="Logistic Function (0...1)"
  }
  n4 = MaterialFunctionCall() {materialFunction=&Pi}
  n5 = Multiply(A=@MaterialExpressionFunctionInput_0, B=Divide(A=@n4.Result) {constA=2})
  FunctionOutput(None=Multiply(A=Arctangent!(None=@n5), B=Divide(B=@n4.Result) {constA=2})) {
   description="Actual Arctangent function. Very soft rolloff, approaches -1/1 boundaries at input values around +/- 300,000\r\n\r\nAdds 24 instructions"
   outputName="Arctangent (-1...1)"
   sortPriority=25
  }
  n6 = MaterialFunctionCall() {materialFunction=&Pi}
  n7 = Multiply(A=@MaterialExpressionFunctionInput_0, B=Divide(A=@n6.Result) {constA=2})
  FunctionOutput(None=Multiply(A=ArctangentFast!(None=@n7), B=Divide(B=@n6.Result) {constA=2})) {
   description="Approximate Arctangent function. Very soft rolloff, approaches -1/1 boundaries at input values around +/- 300,000\r\n\r\nAdds 15 instructions"
   outputName="ArctangentFast (-1...1)"
   sortPriority=20
  }
  FunctionOutput(None=Subtract(A=Divide(B=@n3) {constA=2})) {
   description="Logistic function transformed to output -1...1 space values. (2/(1+e^-x))-1. Soft slope, approaches -1/1 boundary around +/- 5.3\r\n\r\nAdds 6 instructions"
   outputName="Logistic Function (-1...1)"
  }
 ]
}
