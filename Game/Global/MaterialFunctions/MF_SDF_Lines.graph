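# Material function that draws evenly spaced contour lines from an SDF input, with an optional pulse mask.
# Frac repeats the optionally time-scrolled SDF at LineDistance, and a SmoothStep widened by screen-space derivatives gives antialiased Lines.
# Static switches toggle scrolling, pulsing and variable thickness, which remaps a value into a clamped min/max width.
# PulseMask is a Frac sawtooth over floor-quantized bands and time, or 1 when pulsing is off.
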MaterialFunction /Game/Global/MaterialFunctions/MF_SDF_Lines
graph {nodes [
  ComponentMask(None=TextureCoordinate()) {r=true}
  n1 = NamedRerouteUsage()
  n2 = NamedRerouteUsage()
  n3 = Multiply(B=@n2) {constA=0.5}
  n4 = Divide(B=@n2)
  n5 = Floor(None=Multiply(A=Add(A=@n1, B=@n3) {constB=0.2}, B=@n4))
  n6 = Add(
   A=Multiply(A=NamedRerouteUsage(), B=NamedRerouteUsage()) {constB=-1}
   B=Multiply(A=Divide(A=@n5, B=@n4) {constB=10}, B=NamedRerouteUsage()) {constB=20}
  )
  ConstantBiasScale(None=Sine(None=@n6))
  Constant() {r=0.003}
  n7 = FunctionInput() {bUsePreviewValueAsDefault=true, inputName=ThicknessRangeMin, inputType=FunctionInput_Scalar, sortPriority=4, previewValue=<-0.25 0 0 1>}
  n8 = FunctionInput() {bUsePreviewValueAsDefault=true, inputName=ThicknessMin, inputType=FunctionInput_Scalar, sortPriority=6, previewValue=<0.005 0 0 1>}
  n9 = FunctionInput() {bUsePreviewValueAsDefault=true, inputName=ThicknessMax, inputType=FunctionInput_Scalar, sortPriority=7, previewValue=<0.01 0 0 1>}
  MaterialExpressionRemap_3 = MaterialXRemap!(
   None=NamedRerouteUsage()
   InputLow=@n7
   InputHigh=FunctionInput() {bUsePreviewValueAsDefault=true, inputName=ThicknessRangeMax, inputType=FunctionInput_Scalar, sortPriority=5}
   TargetLow=@n8
   TargetHigh=@n9
  ) {inputHighDefault=0, inputLowDefault=-0.25, targetHighDefault=0.001, targetLowDefault=0.02}
  n10 = StaticSwitch(True=Clamp(None=@MaterialExpressionRemap_3, Min=@n8, Max=@n9), False=NamedRerouteUsage(), Value=NamedRerouteUsage())
  n11 = Multiply(A=NamedRerouteUsage(), B=NamedRerouteUsage())
  n12 = NamedRerouteUsage()
  n13 = NamedRerouteDeclaration(None=StaticSwitch(True=Add(A=@n11, B=@n12), False=@n12, Value=NamedRerouteUsage())) {name=SDFModified, nodeColor=<rgba 0 0.6953 1 1>}
  n14 = DDX(Value=@n13)
  n15 = DDY(Value=@n13)
  n16 = Multiply(A=Add(A=Abs(None=@n14), B=Abs(None=@n15))) {constB=0.5}
  n17 = Subtract(A=@n10, B=@n16)
  n18 = Add(A=@n10, B=@n16)
  n19 = NamedRerouteUsage()
  n20 = Frac(None=Add(A=Divide(A=@n13, B=@n19)) {constB=0.5})
  n21 = Multiply(A=Abs(None=Subtract(A=@n20) {constB=0.5}), B=@n19)
  n22 = Frac(None=@n6)
  n23 = Constant() {r=1}
  n24 = NamedRerouteUsage()
  Multiply(
   A=FunctionOutput(None=SmoothStep(Min=@n17, Max=@n18, Value=@n21)) {outputName=Lines}
   B=FunctionOutput(None=StaticSwitch(True=@n22, False=@n23, Value=@n24)) {outputName=PulseMask, sortPriority=1}
  )
  NamedRerouteDeclaration(None=FunctionInput(Preview=MaterialFunctionCall()) {bUsePreviewValueAsDefault=true, inputName=SDF, inputType=FunctionInput_Scalar}) {name=SDF, nodeColor=<rgba 0 0.4609 1 1>}
  n25 = FunctionInput() {bUsePreviewValueAsDefault=true, inputName=LineThickness, inputType=FunctionInput_Scalar, sortPriority=1, previewValue=<0.01 0 0 1>}
  NamedRerouteDeclaration(None=@n25) {name=LineThickness, nodeColor=<rgba 1 0 0.5625 1>}
  n26 = FunctionInput() {bUsePreviewValueAsDefault=true, inputName=LineDistance, inputType=FunctionInput_Scalar, sortPriority=2, previewValue=<0.2 0 0 1>}
  NamedRerouteDeclaration(None=@n26) {name=LineDistance, nodeColor=<rgba 0 1 0.5547 1>}
  NamedRerouteDeclaration(None=FunctionInput(Preview=StaticBool()) {bUsePreviewValueAsDefault=true, inputName=VariableThickness, inputType=FunctionInput_StaticBool, sortPriority=3}) {name=VariableThickness, nodeColor=<rgba 0 0.4609 1 1>}
  NamedRerouteDeclaration(None=FunctionInput(Preview=StaticBool()) {bUsePreviewValueAsDefault=true, inputName=Animate, inputType=FunctionInput_StaticBool, sortPriority=8}) {name=Animate, nodeColor=<rgba 0 0.4609 1 1>}
  n27 = FunctionInput() {bUsePreviewValueAsDefault=true, inputName=Speed, inputType=FunctionInput_Scalar, sortPriority=9, previewValue=<-0.1 0 0 1>}
  NamedRerouteDeclaration(None=@n27) {name=Speed, nodeColor=<rgba 0 1 0.5547 1>}
  NamedRerouteDeclaration(None=FunctionInput(Preview=StaticBool()) {bUsePreviewValueAsDefault=true, inputName=Pulse, inputType=FunctionInput_StaticBool, sortPriority=10}) {name=Pulse, nodeColor=<rgba 0 0.7656 1 1>}
  n28 = FunctionInput() {bUsePreviewValueAsDefault=true, inputName=PulseFrequency, inputType=FunctionInput_Scalar, sortPriority=11, previewValue=<3 0 0 1>}
  NamedRerouteDeclaration(None=@n28) {name=PulseFrequency, nodeColor=<rgba 0 0.7656 1 1>}
  n29 = FunctionInput() {bUsePreviewValueAsDefault=true, inputName=PulseSpeed, inputType=FunctionInput_Scalar, sortPriority=12, previewValue=<0.2 0 0 1>}
  NamedRerouteDeclaration(None=@n29) {name=PulseSpeed, nodeColor=<rgba 0 1 0.5547 1>}
  NamedRerouteDeclaration(None=Add(A=Time(), B=FunctionInput() {bUsePreviewValueAsDefault=true, inputName=TimeOffset, inputType=FunctionInput_Scalar})) {name=Time, nodeColor=<rgba 1 0.8189 0 1>}
 ]}
