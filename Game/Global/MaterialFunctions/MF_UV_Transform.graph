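# Material function that rotates, scales and offsets UV coordinates around a pivot, which defaults to the texture center.
# It subtracts the pivot from the UV, applies a hand-built sine/cosine 2D rotation, multiplies by Scale, then adds back the pivot and Offset.
# Scale applies after rotation along the output axes, and Rotation counts in full turns because Sine and Cosine keep their default period.
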MaterialFunction /Game/Global/MaterialFunctions/MF_UV_Transform
graph {nodes [
  n1 = FunctionInput() {bUsePreviewValueAsDefault=true, inputName=Rotation, inputType=FunctionInput_Scalar, sortPriority=2}
  n2 = Cosine(None=@n1)
  n3 = FunctionInput() {bUsePreviewValueAsDefault=true, inputName=Pivot, inputType=FunctionInput_Vector2, sortPriority=4, previewValue=<0.5 0.5 0 1>}
  n4 = Subtract(A=FunctionInput(Preview=TextureCoordinate()) {bUsePreviewValueAsDefault=true, inputName=UV, inputType=FunctionInput_Vector2}, B=@n3)
  n5 = ComponentMask(None=@n4) {r=true}
  n6 = Multiply(A=@n2, B=@n5)
  n7 = Sine(None=@n1)
  n8 = ComponentMask(None=@n4) {g=true}
  n9 = Multiply(A=@n7, B=@n8)
  n10 = Multiply(A=@n2, B=@n8)
  n11 = Multiply(A=@n7, B=@n5)
  n12 = FunctionInput() {bUsePreviewValueAsDefault=true, inputName=Scale, inputType=FunctionInput_Vector2, sortPriority=3, previewValue=<1 1 0 1>}
  n13 = Multiply(A=AppendVector(A=Add(A=@n6, B=@n9), B=Subtract(A=@n10, B=@n11)), B=@n12)
  FunctionOutput(None=Add(
    A=Add(A=@n13, B=@n3) {constB=0}
    B=FunctionInput() {bUsePreviewValueAsDefault=true, inputName=Offset, inputType=FunctionInput_Vector2, sortPriority=1}
   ))
 ]}
