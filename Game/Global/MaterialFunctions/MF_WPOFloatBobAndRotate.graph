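# Material function that adds an optional floating bob and rocking rotation to incoming material attributes via world position offset, behind a static switch.
# Sine-over-time drives rotation about a parameterized axis and instance-space pivot, plus a bob vector scaled by object Z scale.
# A hash of instance position (X twice plus Z, ignoring Y) slightly varies both speeds per instance.
# Normals are corrected for the rotation.
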MaterialFunction /Game/Global/MaterialFunctions/MF_WPOFloatBobAndRotate
graph {nodes [
  n1 = FunctionInput() {inputType=FunctionInput_MaterialAttributes}
  n2 = GetMaterialAttributes(None=@n1) {attributeGetTypes [#1, #2]}
  n3 = VectorParameter() {group=WPO float Bob and Rotate, parameterName=WPO_RotateAxis, sortPriority=2, defaultValue=<rgba 0 0 1 1>}
  n4 = Normalize(VectorInput=@n3.RGB)
  n5 = Time()
  n6 = ScalarParameter() {defaultValue=0.1, group=WPO float Bob and Rotate, parameterName=WPO_Rotate_Speed}
  n7 = Add(A=@n6, B=Multiply(A=Multiply(A=@n6) {constB=0.01}, B=NamedRerouteUsage()))
  n8 = Multiply(A=Sine(None=Multiply(A=@n5, B=@n7)), B=ScalarParameter() {defaultValue=0.01, group=WPO float Bob and Rotate, parameterName=WPO_Rotate_Angle})
  n9 = VectorParameter() {group=WPO float Bob and Rotate, parameterName=PivotPosition, sortPriority=3, defaultValue=<rgba 0 0 0 1>}
  n10 = TransformPosition(None=@n9.RGB) {transformSourceType=TRANSFORMPOSSOURCE_Instance, transformType=TRANSFORMPOSSOURCE_TranslatedWorld}
  n11 = WorldPosition() {worldPositionShaderOffset=WPT_CameraRelativeNoOffsets}
  n12 = ScalarParameter() {defaultValue=0.2, group=WPO float Bob and Rotate, parameterName=WPO_Float Bobbing Speed}
  n13 = Multiply(A=@n12) {constB=0.01}
  n14 = NamedRerouteUsage()
  n15 = Multiply(A=Time(), B=Add(A=@n12, B=Multiply(A=@n13, B=@n14)))
  n16 = VectorParameter() {group=WPO float Bob and Rotate, parameterName=WPO_Float Bobbing Distance, sortPriority=2, defaultValue=<rgba 0 0 10 1>}
  n17 = MaterialFunctionCall() {materialFunction=&ObjectScale}
  n18 = Add(
   A=Add(A=@n2.WorldPositionOffset, B=RotateAboutAxis(NormalizedRotationAxis=@n4, RotationAngle=@n8, PivotPoint=@n10, Position=@n11.XYZ))
   B=Multiply(A=Sine(None=@n15), B=Multiply(A=@n16.RGB, B=@n17.Scale Z))
  )
  n19 = MaterialFunctionCall(Rotation Axis=@n4, Rotation Angle=@n8, World Space Vertex Normals=@n2.Normal) {materialFunction=&FixRotateAboutAxisNormals}
  FunctionOutput(None=StaticSwitchParameter(
    True=SetMaterialAttributes!(MaterialAttributes=@n2.MaterialAttributes, WorldPositionOffset=@n18, Normal=@n19.New World Space Vertex Normal) {attributeSetTypes [#1, #2]}
    False=@n1
   ) {group=WPO float Bob and Rotate, parameterName=UseWPO_Float Bob and Rotate, sortPriority=0})
  n20 = Constant3Vector()
  n21 = MaterialFunctionCall(Float3=Divide(A=TransformPosition(None=@n20) {transformSourceType=TRANSFORMPOSSOURCE_Instance, transformType=TRANSFORMPOSSOURCE_World}) {constB=100}) {materialFunction=&BreakOutFloat3Components}
  n22 = Add(A=@n21.R, B=@n21.R)
  NamedRerouteDeclaration(None=Frac(None=Add(A=@n22, B=@n21.B))) {name=Random, nodeColor=<rgba 0.6406 1 0 1>}
 ]}
