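# Library material function that projects a tiling normal map triplanar-style from world position, scaled by Size.
# Each of three plane samples is sign-corrected, then Whiteout-blended with swizzled, vertex-interpolated pre-skinned normals using NormalFlatness.
# Contrast-sharpened world-normal X and Z masks lerp the projections, and the result is transformed from local to tangent space for XYZ Texture.
# ZTexture exposes the Z projection alone, without that transform.
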MaterialFunction /Game/Global/MaterialFunctions/MF_WorldAlignedNormal_Whiteout
graph {
 properties {bExposeToLibrary=true, description=HIGH QUALITY world aligned normal map blending, libraryCategoriesText [WorldAligned]}
 nodes [
  n1 = WorldPosition()
  n2 = FunctionInput() {bUsePreviewValueAsDefault=true, inputName=Size, inputType=FunctionInput_Scalar, previewValue=<500 500 500 1>}
  n3 = Abs(None=@n2)
  n4 = MaterialFunctionCall(Float3=Divide(A=FunctionInput(Preview=@n1.XYZ) {bUsePreviewValueAsDefault=true, inputName=Position}, B=Multiply(A=@n3) {constB=-1})) {materialFunction=&BreakOutFloat3Components}
  n5 = MaterialFunctionCall(X=@n4.R, Y=@n4.R) {desc=Y facing Plane, materialFunction=&MakeFloat2}
  Add(A=@n5.Result, B=Constant2Vector() {g=0.5})
  n6 = MaterialFunctionCall(X=@n4.R, Y=@n4.R) {desc=Z facing Plane, materialFunction=&MakeFloat2}
  n7 = FunctionInput(Preview=TextureObject() {samplerType=SAMPLERTYPE_Normal, texture=/Engine/EngineMaterials/FlatNormal}) {bUsePreviewValueAsDefault=true, inputName=TextureObject, inputType=FunctionInput_Texture2D, sortPriority=-10}
  n8 = FunctionInput() {bUsePreviewValueAsDefault=true, inputName=NormalFlatness, inputType=FunctionInput_Scalar}
  n9 = VertexInterpolator(VS=PreSkinnedNormal())
  n10 = MaterialFunctionCall!(
   NormalTexture=Multiply(A=Constant3Vector() {constant=<rgba -1 -1 1 0>}, B=TextureSample(UVs=@n6.Result, Tex=@n7) {samplerSource=SSM_Wrap_WorldGroupSettings}.RGB)
   NormalFlatness=@n8
   WorldNormals=@n9.PS
  ) {materialFunction=&/MF_BlendAxisNormals_Whiteout}
  FunctionOutput(None=@n10.Result) {outputName=ZTexture}
  n11 = Constant3Vector() {constant=<rgba -1 1 1 1>}
  n12 = TextureSample(UVs=@n5.Result, Tex=@n7) {samplerSource=SSM_Wrap_WorldGroupSettings}
  n13 = MaterialFunctionCall(Float3=@n9.PS) {materialFunction=&BreakOutFloat3Components}
  n14 = Multiply(A=@n13.B) {constB=-1}
  n15 = MaterialFunctionCall(Float3=MaterialFunctionCall(
    NormalTexture=Multiply(A=@n11, B=@n12.RGB)
    NormalFlatness=@n8
    WorldNormals=MaterialFunctionCall(X=@n13.R, Y=@n14, Z=@n13.R) {materialFunction=&MakeFloat3}.Result
   ) {materialFunction=&MF_BlendAxisNormals_Whiteout}.Result) {materialFunction=&BreakOutFloat3Components}
  n16 = Constant3Vector() {constant=<rgba -1 1 1 1>}
  n17 = TextureSample(
   UVs=MaterialFunctionCall(X=@n4.G, Y=@n4.G) {desc=X facing Plane, materialFunction=&MakeFloat2}.Result
   Tex=@n7
  ) {samplerSource=SSM_Wrap_WorldGroupSettings}
  n18 = Multiply(A=@n13.B) {constB=-1}
  n19 = MaterialFunctionCall(Float3=MaterialFunctionCall(
    NormalTexture=Multiply(A=@n16, B=@n17.RGB)
    NormalFlatness=@n8
    WorldNormals=MaterialFunctionCall(X=@n13.G, Y=@n18, Z=@n13.G) {materialFunction=&MakeFloat3}.Result
   ) {materialFunction=&MF_BlendAxisNormals_Whiteout}.Result) {materialFunction=&BreakOutFloat3Components}
  n20 = Multiply(A=FunctionInput(Preview=VertexNormalWS()) {bUsePreviewValueAsDefault=true, inputName=NormalWS, sortPriority=-7}, B=Constant() {r=1})
  n21 = ComponentMask(None=@n20) {r=true}
  n22 = FunctionInput() {
   bUsePreviewValueAsDefault=true
   inputName=ProjectionTransitionSharpness
   inputType=FunctionInput_Scalar
   sortPriority=-6
   previewValue=<16 0 0 1>
  }
  n23 = LinearInterpolate(
   A=MaterialFunctionCall(X=@n15.R, Y=@n15.R, Z=Multiply(A=@n15.G) {constB=-1}) {materialFunction=&MakeFloat3}.Result
   B=MaterialFunctionCall(X=@n19.B, Y=@n19.B, Z=Multiply(A=@n19.G) {constB=-1}) {materialFunction=&MakeFloat3}.Result
   Alpha=MaterialFunctionCall(In=Abs(None=@n21), Contrast=@n22) {materialFunction=&CheapContrast}.Result
  )
  n24 = MaterialFunctionCall(In=Abs(None=ComponentMask(None=@n20) {b=true}), Contrast=@n22) {materialFunction=&CheapContrast}
  FunctionOutput(None=Transform(None=LinearInterpolate(A=@n23, B=@n10.Result, Alpha=@n24.Result)) {transformSourceType=TRANSFORMSOURCE_Local, transformType=TRANSFORM_Tangent}) {outputName=XYZ Texture}
 ]
}
